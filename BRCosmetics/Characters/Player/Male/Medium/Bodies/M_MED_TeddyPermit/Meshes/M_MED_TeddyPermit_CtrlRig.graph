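# Control Rig for the TeddyPermit male player cosmetic that drives coat, sleeve, shoe and ribbon secondary bones procedurally on top of a FortDeformRig pass.
# A shared DeltaRotation function converts swing or twist of thighs, spine and arms into clamped weights for corrective bone offsets, multiplying pairs for combined poses.
# Crouching, sliding and skydiving flags are Kalman-filtered into pose blends, and cloth_sim adjusts spine-twist offsets.
# Shoe pivots and sandals follow leg bones through constraints.
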ControlRigBlueprint /BRCosmetics/Characters/Player/Male/Medium/Bodies/M_MED_TeddyPermit/Meshes/M_MED_TeddyPermit_CtrlRig
graphs [
 New Function_ContainedGraph {nodes [
   RigVMFunctionEntryNode()
   > RigVMFunction_Sequence_1 = RigVMAggregateNode()
   DeltaRotation v5 = RigVMFunctionReferenceNode(
    Rotation Order=ZYX
    child=(Type=Bone,Name="thigh_l")
    outputMinMax=(X=0,Y=1)
    parent=(Type=Bone,Name="thigh_l")
    premult=1
    rotationMinMax=(X=0,Y=-120)
    triggerAxis=Z
    twistAxis=(X=1,Y=0,Z=0)
    ExecuteContext=@RigVMFunction_Sequence_1.A
   )
   > RigVMVariableNode(Value=@"DeltaRotation v5".outputWeight) {variable=L_leg_up}
   > DeltaRotation v5_1 = RigVMFunctionReferenceNode(
    Rotation Order=ZYX
    child=(Type=Bone,Name="thigh_l")
    outputMinMax=(X=0,Y=1)
    parent=(Type=Bone,Name="thigh_l")
    premult=1
    rotationMinMax=(X=12,Y=55)
    triggerAxis=Z
    twistAxis=(X=1,Y=0,Z=0)
   )
   > RigVMVariableNode(Value=@"DeltaRotation v5_1".outputWeight) {variable=L_leg_back}
   > DeltaRotation v5_2 = RigVMFunctionReferenceNode(
    Rotation Order=ZYX
    child=(Type=Bone,Name="thigh_l")
    outputMinMax=(X=0,Y=1)
    parent=(Type=Bone,Name="thigh_l")
    premult=1
    rotationMinMax=(X=0,Y=-85)
    triggerAxis=Y
    twistAxis=(X=1,Y=0,Z=0)
   )
   > RigVMVariableNode(Value=@"DeltaRotation v5_2".outputWeight) {variable=L_leg_out}
   > DeltaRotation v5_3 = RigVMFunctionReferenceNode(
    Rotation Order=ZYX
    child=(Type=Bone,Name="thigh_l")
    outputMinMax=(X=0,Y=1)
    parent=(Type=Bone,Name="thigh_l")
    premult=1
    rotationMinMax=(X=0,Y=30)
    triggerAxis=Y
    twistAxis=(X=1,Y=0,Z=0)
   )
   > RigVMVariableNode(Value=@"DeltaRotation v5_3".outputWeight) {variable=L_leg_in}
   > DeltaRotation v5_4 = RigVMFunctionReferenceNode(
    Rotation Order=ZYX
    child=(Type=Bone,Name="thigh_l")
    onlyTwist=True
    outputMinMax=(X=0,Y=1)
    parent=(Type=Bone,Name="thigh_l")
    premult=1
    rotationMinMax=(X=0,Y=-100)
    triggerAxis=X
    twistAxis=(X=1,Y=0,Z=0)
   )
   > RigVMVariableNode(Value=@"DeltaRotation v5_4".outputWeight) {variable=L_leg_twist_out}
   > DeltaRotation v5_5 = RigVMFunctionReferenceNode(
    Rotation Order=ZYX
    child=(Type=Bone,Name="thigh_l")
    onlyTwist=True
    outputMinMax=(X=0,Y=1)
    parent=(Type=Bone,Name="thigh_l")
    premult=1
    rotationMinMax=(X=0,Y=45)
    triggerAxis=X
    twistAxis=(X=1,Y=0,Z=0)
   )
   > RigVMVariableNode(Value=@"DeltaRotation v5_5".outputWeight) {variable=L_leg_twist_in}
   DeltaRotation v5_6 = RigVMFunctionReferenceNode(
    Rotation Order=ZYX
    child=(Type=Bone,Name="thigh_r")
    ignore initial offset=false
    outputMinMax=(X=0,Y=1)
    parent=(Type=Bone,Name="thigh_r")
    premult=1
    rotationMinMax=(X=0,Y=-120)
    triggerAxis=Z
    twistAxis=(X=1,Y=0,Z=0)
    ExecuteContext=@RigVMFunction_Sequence_1.B
   )
   > RigVMVariableNode(Value=@"DeltaRotation v5_6".outputWeight) {variable=R_leg_up}
   > DeltaRotation v5_7 = RigVMFunctionReferenceNode(
    Rotation Order=ZYX
    child=(Type=Bone,Name="thigh_r")
    outputMinMax=(X=0,Y=1)
    parent=(Type=Bone,Name="thigh_r")
    premult=1
    rotationMinMax=(X=12,Y=55)
    triggerAxis=Z
    twistAxis=(X=1,Y=0,Z=0)
   )
   > RigVMVariableNode(Value=@"DeltaRotation v5_7".outputWeight) {variable=R_leg_back}
   > DeltaRotation v5_8 = RigVMFunctionReferenceNode(
    Rotation Order=ZYX
    child=(Type=Bone,Name="thigh_r")
    outputMinMax=(X=0,Y=1)
    parent=(Type=Bone,Name="thigh_r")
    premult=1
    rotationMinMax=(X=0,Y=-85)
    triggerAxis=Y
    twistAxis=(X=1,Y=0,Z=0)
   )
   > RigVMVariableNode(Value=@"DeltaRotation v5_8".outputWeight) {variable=R_leg_out}
   > DeltaRotation v5_9 = RigVMFunctionReferenceNode(
    Rotation Order=ZYX
    child=(Type=Bone,Name="thigh_r")
    outputMinMax=(X=0,Y=1)
    parent=(Type=Bone,Name="thigh_r")
    premult=1
    rotationMinMax=(X=0,Y=30)
    triggerAxis=Y
    twistAxis=(X=1,Y=0,Z=0)
   )
   > RigVMVariableNode(Value=@"DeltaRotation v5_9".outputWeight) {variable=R_leg_in}
   > DeltaRotation v5_10 = RigVMFunctionReferenceNode(
    Rotation Order=ZYX
    child=(Type=Bone,Name="thigh_r")
    onlyTwist=True
    outputMinMax=(X=0,Y=1)
    parent=(Type=Bone,Name="thigh_r")
    premult=1
    rotationMinMax=(X=0,Y=-100)
    triggerAxis=X
    twistAxis=(X=1,Y=0,Z=0)
   )
   > RigVMVariableNode(Value=@"DeltaRotation v5_10".outputWeight) {variable=R_leg_twist_out}
   > DeltaRotation v5_11 = RigVMFunctionReferenceNode(
    Rotation Order=ZYX
    child=(Type=Bone,Name="thigh_r")
    onlyTwist=True
    outputMinMax=(X=0,Y=1)
    parent=(Type=Bone,Name="thigh_r")
    premult=1
    rotationMinMax=(X=0,Y=45)
    triggerAxis=X
    twistAxis=(X=1,Y=0,Z=0)
   )
   > RigVMVariableNode(Value=@"DeltaRotation v5_11".outputWeight) {variable=R_leg_twist_in}
   DeltaRotation v5_24 = RigVMFunctionReferenceNode(
    Rotation Order=ZYX
    child=(Type=Bone,Name="spine_03")
    ignore initial offset=true
    outputMinMax=(X=0,Y=1)
    parent=(Type=Bone,Name="pelvis")
    premult=1
    rotationMinMax=(X=0,Y=60)
    triggerAxis=Z
    twistAxis=(X=1,Y=0,Z=0)
    ExecuteContext=@RigVMFunction_Sequence_1.C
   )
   > RigVMVariableNode(Value=@"DeltaRotation v5_24".outputWeight) {variable=spine_forward}
   > DeltaRotation v5_13 = RigVMFunctionReferenceNode(
    Rotation Order=ZYX
    child=(Type=Bone,Name="spine_03")
    ignore initial offset=True
    outputMinMax=(X=0,Y=1)
    parent=(Type=Bone,Name="pelvis")
    premult=1
    rotationMinMax=(X=0,Y=-30)
    triggerAxis=Z
    twistAxis=(X=1,Y=0,Z=0)
   )
   > RigVMVariableNode(Value=@"DeltaRotation v5_13".outputWeight) {variable=spine_backward}
   > DeltaRotation v5_14 = RigVMFunctionReferenceNode(
    Rotation Order=ZYX
    child=(Type=Bone,Name="spine_03")
    ignore initial offset=True
    outputMinMax=(X=0,Y=1)
    parent=(Type=Bone,Name="pelvis")
    premult=1
    rotationMinMax=(X=0,Y=-45)
    triggerAxis=Y
    twistAxis=(X=1,Y=0,Z=0)
   )
   > RigVMVariableNode(Value=@"DeltaRotation v5_14".outputWeight) {variable=spine_right}
   > DeltaRotation v5_15 = RigVMFunctionReferenceNode(
    Rotation Order=ZYX
    child=(Type=Bone,Name="spine_03")
    ignore initial offset=True
    outputMinMax=(X=0,Y=1)
    parent=(Type=Bone,Name="pelvis")
    premult=1
    rotationMinMax=(X=0,Y=45)
    triggerAxis=Y
    twistAxis=(X=1,Y=0,Z=0)
   )
   > RigVMVariableNode(Value=@"DeltaRotation v5_15".outputWeight) {variable=spine_left}
   > DeltaRotation v5_16 = RigVMFunctionReferenceNode(
    Rotation Order=ZYX
    child=(Type=Bone,Name="spine_04")
    ignore initial offset=false
    outputMinMax=(X=0,Y=1)
    parent=(Type=Bone,Name="spine_04")
    premult=1
    rotationMinMax=(X=0,Y=60)
    triggerAxis=Z
    twistAxis=(X=1,Y=0,Z=0)
   )
   > RigVMVariableNode(Value=@"DeltaRotation v5_16".outputWeight) {variable=upper_spine_forward}
   > DeltaRotation v5_17 = RigVMFunctionReferenceNode(
    Rotation Order=ZYX
    child=(Type=Bone,Name="spine_04")
    outputMinMax=(X=0,Y=1)
    parent=(Type=Bone,Name="spine_04")
    premult=1
    rotationMinMax=(X=0,Y=-45)
    triggerAxis=Z
    twistAxis=(X=1,Y=0,Z=0)
   )
   > RigVMVariableNode(Value=@"DeltaRotation v5_17".outputWeight) {variable=upper_spine_backward}
   > DeltaRotation v5_18 = RigVMFunctionReferenceNode(
    Rotation Order=ZYX
    child=(Type=Bone,Name="spine_04")
    outputMinMax=(X=0,Y=1)
    parent=(Type=Bone,Name="spine_04")
    premult=1
    rotationMinMax=(X=0,Y=-30)
    triggerAxis=Y
    twistAxis=(X=1,Y=0,Z=0)
   )
   > RigVMVariableNode(Value=@"DeltaRotation v5_18".outputWeight) {variable=upper_spine_right}
   > DeltaRotation v5_19 = RigVMFunctionReferenceNode(
    Rotation Order=ZYX
    child=(Type=Bone,Name="spine_04")
    outputMinMax=(X=0,Y=1)
    parent=(Type=Bone,Name="spine_04")
    premult=1
    rotationMinMax=(X=0,Y=30)
    triggerAxis=Y
    twistAxis=(X=1,Y=0,Z=0)
   )
   > RigVMVariableNode(Value=@"DeltaRotation v5_19".outputWeight) {variable=upper_spine_left}
   > DeltaRotation v5_20 = RigVMFunctionReferenceNode(
    Rotation Order=ZYX
    child=(Type=Bone,Name="spine_04")
    onlyTwist=true
    outputMinMax=(X=0,Y=1)
    parent=(Type=Bone,Name="spine_04")
    premult=1
    rotationMinMax=(X=0,Y=30)
    triggerAxis=X
    twistAxis=(X=1,Y=0,Z=0)
   )
   > RigVMVariableNode(Value=@"DeltaRotation v5_20".outputWeight) {variable=upper_spine_twist_out}
   > DeltaRotation v5_21 = RigVMFunctionReferenceNode(
    Rotation Order=ZYX
    child=(Type=Bone,Name="spine_04")
    onlyTwist=true
    outputMinMax=(X=0,Y=1)
    parent=(Type=Bone,Name="spine_04")
    premult=1
    rotationMinMax=(X=0,Y=-30)
    triggerAxis=X
    twistAxis=(X=1,Y=0,Z=0)
   )
   > RigVMVariableNode(Value=@"DeltaRotation v5_21".outputWeight) {variable=upper_spine_twist_in}
   n1 = RigVMVariableNode() {variable=L_leg_up}
   &RigUnit_ModifyTransforms(
    ItemToModify=((Item=(Type=Bone,Name="dyn_coat_01_L_02"),Transform=(Rotation=(X=0.3827,Y=0,Z=0,W=0.9239),Translation=(X=0,Y=0,Z=-1.5),Scale3D=(X=1,Y=1,Z=1))))
    WeightMaximum=0.5
    WeightMinimum=0
    ExecutePin=@RigVMFunction_Sequence_1.D
    Weight=@n1.Value
   )
   n2 = RigVMVariableNode() {variable=R_leg_up}
   > &RigUnit_ModifyTransforms(
    ItemToModify=((Item=(Type=Bone,Name="dyn_coat_01_R_02"),Transform=(Rotation=(X=-0.3827,Y=0,Z=0,W=0.9239),Translation=(X=0,Y=0,Z=1.5),Scale3D=(X=1,Y=1,Z=1))))
    WeightMaximum=0.5
    WeightMinimum=0
    Weight=@n2.Value
   )
   n3 = RigVMVariableNode() {variable=L_leg_back}
   > &RigUnit_ModifyTransforms(
    ItemToModify="((Item=(Type=Bone,Name=\"dyn_coat_03_L_01\"),Transform=(Rotation=(X=-0.369,Y=0.0443,Z=-0.1106,W=0.9218),Translation=(X=0,Y=0,Z=0),Scale3D=(X=1,Y=1,Z=1))),(Item=(Type=Bone,Name=\"dyn_coat_B_01\"),Transform=(Rotation=(X=0,Y=0,Z=-0.1177,W=0.993),Translation=(X=0, ...(truncated)"
    WeightMinimum=0
    Weight=@n3.Value
   )
   n4 = RigVMVariableNode() {variable=R_leg_back}
   > &RigUnit_ModifyTransforms(
    ItemToModify="((Item=(Type=Bone,Name=\"dyn_coat_03_R_01\"),Transform=(Rotation=(X=0.369,Y=-0.0443,Z=-0.1106,W=0.9218),Translation=(X=0,Y=0,Z=0),Scale3D=(X=1,Y=1,Z=1))),(Item=(Type=Bone,Name=\"dyn_coat_B_01\"),Transform=(Rotation=(X=0,Y=0,Z=-0.1177,W=0.993),Translation=(X=0, ...(truncated)"
    WeightMinimum=0
    Weight=@n4.Value
   )
   Multiply = &RigVMFunction_MathDoubleMul(B=@n4.Value, A=@n3.Value)
   > &RigUnit_ModifyTransforms(
    ItemToModify="((Item=(Type=Bone,Name=\"dyn_coat_B_01\"),Transform=(Rotation=(X=0,Y=0,Z=0.1177,W=0.993),Translation=(X=0,Y=0,Z=0),Scale3D=(X=1,Y=1,Z=1))),(Item=(Type=Bone,Name=\"dyn_coat_B_02\"),Transform=(Rotation=(X=0,Y=0,Z=0.2986,W=0.9544),Translation=(X=0,Y=0,Z=0),Scale3 ...(truncated)"
    WeightMinimum=0
    Weight=@Multiply.Result
   )
   n5 = RigVMVariableNode() {variable=L_leg_out}
   > &RigUnit_ModifyTransforms(
    ItemToModify="((Item=(Type=Bone,Name=\"dyn_coat_02_L_02\"),Transform=(Rotation=(X=0,Y=0,Z=-0.6735,W=0.7392),Translation=(X=-8.4176,Y=-7.4827,Z=0),Scale3D=(X=1,Y=1,Z=1))),(Item=(Type=Bone,Name=\"dyn_coat_02_L_01\"),Transform=(Rotation=(X=0,Y=0,Z=0,W=1),Translation=(X=0,Y=0,Z ...(truncated)"
    WeightMinimum=0
    Weight=@n5.Value
   )
   n6 = RigVMVariableNode() {variable=R_leg_out}
   > &RigUnit_ModifyTransforms(
    ItemToModify="((Item=(Type=Bone,Name=\"dyn_coat_02_R_02\"),Transform=(Rotation=(X=0,Y=0,Z=-0.6735,W=0.7392),Translation=(X=-8.4176,Y=-7.4827,Z=0),Scale3D=(X=1,Y=1,Z=1))),(Item=(Type=Bone,Name=\"dyn_coat_02_R_01\"),Transform=(Rotation=(X=0,Y=0,Z=0,W=1),Translation=(X=0,Y=0,Z ...(truncated)"
    WeightMinimum=0
    Weight=@n6.Value
   )
   > &RigUnit_ModifyTransforms(
    ItemToModify="((Item=(Type=Bone,Name=\"dyn_coat_01_L_03\"),Transform=(Rotation=(X=-0.2158,Y=0.1091,Z=0.3026,W=-0.9219),Translation=(X=-10.449,Y=3.0964,Z=-2.7623),Scale3D=(X=1,Y=1,Z=1))),(Item=(Type=Bone,Name=\"dyn_coat_01_L_02\"),Transform=(Rotation=(X=0,Y=0,Z=0.4354,W=-0.9 ...(truncated)"
    WeightMaximum=0.5
    WeightMinimum=0
    Weight=&RigVMFunction_MathDoubleMul(A=RigVMVariableNode() {variable=L_leg_up}.Value, B=RigVMVariableNode() {variable=L_leg_out}.Value).Result
   )
   > &RigUnit_ModifyTransforms(
    ItemToModify="((Item=(Type=Bone,Name=\"dyn_coat_01_R_03\"),Transform=(Rotation=(X=-0.2158,Y=0.1091,Z=-0.3026,W=0.9219),Translation=(X=-10.449,Y=3.0964,Z=2.7623),Scale3D=(X=1,Y=1,Z=1))),(Item=(Type=Bone,Name=\"dyn_coat_01_R_02\"),Transform=(Rotation=(X=0,Y=0,Z=0.4354,W=-0.90 ...(truncated)"
    WeightMaximum=0.5
    WeightMinimum=0
    Weight=&RigVMFunction_MathDoubleMul(B=RigVMVariableNode() {variable=R_leg_out}.Value, A=RigVMVariableNode() {variable=R_leg_up}.Value).Result
   )
   > &RigUnit_ModifyTransforms(
    ItemToModify="((Item=(Type=Bone,Name=\"dyn_coat_02_L_01\"),Transform=(Rotation=(X=0.0974,Y=0.0747,Z=-0.1371,W=0.9829),Translation=(X=0,Y=0,Z=0),Scale3D=(X=1,Y=1,Z=1))),(Item=(Type=Bone,Name=\"dyn_coat_01_L_02\"),Transform=(Rotation=(X=-0.3225,Y=0.2525,Z=-0.1713,W=0.896),Tra ...(truncated)"
    WeightMinimum=0
    Weight=&RigVMFunction_MathDoubleMul(A=@n1.Value, B=@n5.Value).Result
   )
   > &RigUnit_ModifyTransforms(
    ItemToModify="((Item=(Type=Bone,Name=\"dyn_coat_02_R_01\"),Transform=(Rotation=(X=-0.0974,Y=-0.0747,Z=-0.1371,W=0.9829),Translation=(X=0,Y=0,Z=0),Scale3D=(X=1,Y=1,Z=1))),(Item=(Type=Bone,Name=\"dyn_coat_01_R_02\"),Transform=(Rotation=(X=0.3225,Y=-0.2525,Z=-0.1713,W=0.896),T ...(truncated)"
    WeightMinimum=0
    Weight=&RigVMFunction_MathDoubleMul(B=@n6.Value, A=@n2.Value).Result
   )
   > &RigUnit_ModifyTransforms(
    ItemToModify="((Item=(Type=Bone,Name=\"dyn_coat_01_L_01\"),Transform=(Rotation=(X=0,Y=0,Z=-0.3007,W=0.9537),Translation=(X=0,Y=0,Z=0),Scale3D=(X=1,Y=1,Z=1))),(Item=(Type=Bone,Name=\"dyn_coat_01_R_01\"),Transform=(Rotation=(X=0,Y=0,Z=-0.3007,W=0.9537),Translation=(X=0,Y=0,Z= ...(truncated)"
    WeightMaximum=0.8
    WeightMinimum=0
    Weight=RigVMVariableNode() {variable=spine_forward}.Value
   )
   > &RigUnit_ModifyTransforms(
    ItemToModify="((Item=(Type=Bone,Name=\"dyn_coat_01_L_01\"),Transform=(Rotation=(X=-0.0653,Y=-0.113,Z=-0.4957,W=0.8586),Translation=(X=0,Y=0,Z=0),Scale3D=(X=1,Y=1,Z=1))),(Item=(Type=Bone,Name=\"dyn_coat_01_R_01\"),Transform=(Rotation=(X=0.0653,Y=0.113,Z=-0.4957,W=0.8586),Tra ...(truncated)"
    WeightMinimum=0
    Weight=RigVMVariableNode() {variable=upper_spine_forward}.Value
   )
   > &RigUnit_ModifyTransforms(
    ItemToModify="((Item=(Type=Bone,Name=\"dyn_coat_01_L_01\"),Transform=(Rotation=(X=-0.0828,Y=0.2,Z=0.3736,W=0.902),Translation=(X=0,Y=0,Z=0),Scale3D=(X=1,Y=1,Z=1))),(Item=(Type=Bone,Name=\"dyn_coat_01_R_01\"),Transform=(Rotation=(X=0.0828,Y=-0.2,Z=0.3736,W=0.902),Translation ...(truncated)"
    WeightMinimum=0
    Weight=RigVMVariableNode() {variable=spine_backward}.Value
   )
   > &RigUnit_ModifyTransforms(
    ItemToModify="((Item=(Type=Bone,Name=\"dyn_coat_01_L_01\"),Transform=(Rotation=(X=-0.0446,Y=0.1227,Z=0.3391,W=0.9317),Translation=(X=0,Y=0,Z=0),Scale3D=(X=1,Y=1,Z=1))),(Item=(Type=Bone,Name=\"dyn_coat_01_R_01\"),Transform=(Rotation=(X=0.0446,Y=-0.1227,Z=0.3391,W=0.9317),Tra ...(truncated)"
    WeightMinimum=0
    Weight=RigVMVariableNode() {variable=upper_spine_backward}.Value
   )
   > &RigUnit_ModifyTransforms(
    ItemToModify="((Item=(Type=Bone,Name=\"dyn_coat_01_L_01\"),Transform=(Rotation=(X=0.0522,Y=0.2961,Z=-0.1656,W=0.9392),Translation=(X=0,Y=0,Z=0),Scale3D=(X=1,Y=1,Z=1))),(Item=(Type=Bone,Name=\"dyn_coat_01_R_01\"),Transform=(Rotation=(X=-0.0522,Y=0.2961,Z=0.1656,W=0.9392),Tra ...(truncated)"
    WeightMinimum=0
    Weight=RigVMVariableNode() {variable=spine_left}.Value
   )
   > &RigUnit_ModifyTransforms(
    ItemToModify="((Item=(Type=Bone,Name=\"dyn_coat_01_L_01\"),Transform=(Rotation=(X=0.0157,Y=0.2455,Z=-0.0618,W=0.9673),Translation=(X=0,Y=0,Z=0),Scale3D=(X=1,Y=1,Z=1))),(Item=(Type=Bone,Name=\"dyn_coat_01_R_01\"),Transform=(Rotation=(X=-0.0198,Y=0.2325,Z=0.0827,W=0.9689),Tra ...(truncated)"
    WeightMinimum=0
    Weight=RigVMVariableNode() {variable=upper_spine_left}.Value
   )
   > &RigUnit_ModifyTransforms(
    ItemToModify="((Item=(Type=Bone,Name=\"dyn_coat_01_L_01\"),Transform=(Rotation=(X=0.0522,Y=-0.2961,Z=0.1656,W=0.9392),Translation=(X=0,Y=0,Z=0),Scale3D=(X=1,Y=1,Z=1))),(Item=(Type=Bone,Name=\"dyn_coat_01_R_01\"),Transform=(Rotation=(X=-0.0522,Y=-0.2961,Z=-0.1656,W=0.9392),T ...(truncated)"
    WeightMinimum=0
    Weight=RigVMVariableNode() {variable=spine_right}.Value
   )
   > &RigUnit_ModifyTransforms(
    ItemToModify="((Item=(Type=Bone,Name=\"dyn_coat_01_L_01\"),Transform=(Rotation=(X=0.0157,Y=-0.2455,Z=0.0618,W=0.9673),Translation=(X=0,Y=0,Z=0),Scale3D=(X=1,Y=1,Z=1))),(Item=(Type=Bone,Name=\"dyn_coat_01_R_01\"),Transform=(Rotation=(X=-0.0198,Y=-0.2325,Z=-0.0827,W=0.9689),T ...(truncated)"
    WeightMinimum=0
    Weight=RigVMVariableNode() {variable=upper_spine_right}.Value
   )
   Cast = &RigVMFunction_MathBoolToFloat(Value=RigVMVariableNode() {variable=cloth_sim}.Value)
   Subtract = &RigVMFunction_MathDoubleSub(B=0, A=&RigVMFunction_MathDoubleMul(B=-10, A=@Cast.Result).Result)
   > &RigUnit_ModifyTransforms(
    ItemToModify="((Item=(Type=Bone,Name=\"dyn_coat_01_L_01\"),Transform=(Rotation=(X=0,Y=0,Z=0,W=1),Translation=(X=0,Y=0,Z=-5),Scale3D=(X=1,Y=1,Z=1))),(Item=(Type=Bone,Name=\"dyn_coat_01_R_01\"),Transform=(Rotation=(X=0,Y=0,Z=0,W=1),Translation=(X=0,Y=-10,Z=-5),Scale3D=(X=1,Y= ...(truncated)"
    WeightMinimum=0
    ItemToModify.0.Transform.Translation.Z=@Subtract.Result
    ItemToModify.1.Transform.Translation.Z=@Subtract.Result
    Weight=RigVMVariableNode() {variable=upper_spine_twist_out}.Value
   )
   Add = &RigVMFunction_MathDoubleAdd(B=5, A=&RigVMFunction_MathDoubleMul(B=5, A=@Cast.Result).Result)
   > &RigUnit_ModifyTransforms(
    ItemToModify="((Item=(Type=Bone,Name=\"dyn_coat_01_L_01\"),Transform=(Rotation=(X=0,Y=0,Z=0,W=1),Translation=(X=0,Y=0,Z=5),Scale3D=(X=1,Y=1,Z=1))),(Item=(Type=Bone,Name=\"dyn_coat_01_R_01\"),Transform=(Rotation=(X=0,Y=0,Z=0,W=-1),Translation=(X=0,Y=0,Z=5),Scale3D=(X=1,Y=1,Z ...(truncated)"
    WeightMinimum=0
    ItemToModify.0.Transform.Translation.Z=@Add.Result
    ItemToModify.1.Transform.Translation.Z=@Add.Result
    Weight=RigVMVariableNode() {variable=upper_spine_twist_in}.Value
   )
   RigVMFunctionReturnNode(ExecuteContext=@RigVMFunction_Sequence_1.F)
   RigVMFunction_MathBoolToFloat = &RigVMFunction_MathBoolToFloat(Value=RigVMVariableNode() {variable=crouching}.Value)
   KalmanFilter = &RigVMFunction_KalmanFloat(BufferSize=10, Value=@RigVMFunction_MathBoolToFloat.Result) {template="KalmanFilter::Execute(in Value,in BufferSize,out Result)"}
   &RigUnit_ModifyTransforms(
    ItemToModify="((Item=(Type=Bone,Name=\"dyn_coat_01_L_02\"),Transform=(Rotation=(X=-0.3694,Y=-0.2314,Z=-0.1014,W=-0.8943),Translation=(X=-3.4099,Y=3.6749,Z=0.1089),Scale3D=(X=1,Y=1,Z=1))),(Item=(Type=Bone,Name=\"dyn_coat_01_L_04\"),Transform=(Rotation=(X=0,Y=0.0494,Z=0,W=-0. ...(truncated)"
    WeightMinimum=0
    Weight=@KalmanFilter.Result
    ExecutePin=@RigVMFunction_Sequence_1.E
   )
   > &RigUnit_ModifyTransforms(
    ItemToModify="((Item=(Type=Bone,Name=\"dyn_coat_01_R_01\"),Transform=(Rotation=(X=0,Y=0,Z=0,W=1),Translation=(X=0,Y=0,Z=5),Scale3D=(X=1,Y=1,Z=1))),(Item=(Type=Bone,Name=\"dyn_coat_02_R_01\"),Transform=(Rotation=(X=0,Y=0,Z=0,W=1),Translation=(X=0,Y=0,Z=7),Scale3D=(X=1,Y=1,Z= ...(truncated)"
    WeightMinimum=0
    Weight=@KalmanFilter.Result
   )
   n7 = RigVMVariableNode() {variable=sliding}
   > &RigUnit_ModifyTransforms(
    ItemToModify="((Item=(Type=Bone,Name=\"dyn_coat_01_L_01\"),Transform=(Rotation=(X=-0.047,Y=0.0899,Z=-0.0196,W=0.9946),Translation=(X=0,Y=0,Z=0),Scale3D=(X=1,Y=1,Z=1))),(Item=(Type=Bone,Name=\"dyn_coat_02_L_01\"),Transform=(Rotation=(X=0.4316,Y=-0.1277,Z=0.095,W=-0.8879),Tra ...(truncated)"
    WeightMinimum=0
    Weight=&RigVMFunction_KalmanFloat(BufferSize=15, Value=&RigVMFunction_MathBoolToFloat(Value=@n7.Value).Result) {template="KalmanFilter::Execute(in Value,in BufferSize,out Result)"}.Result
   )
   n8 = RigVMVariableNode() {variable=skydiving}
   > &RigUnit_ModifyTransforms(
    ItemToModify="((Item=(Type=Bone,Name=\"MSTR_coat_B_01\"),Transform=(Rotation=(X=0,Y=0,Z=-0.0872,W=0.9962),Translation=(X=0,Y=-2,Z=0),Scale3D=(X=1,Y=1,Z=1))),(Item=(Type=Bone,Name=\"MSTR_coat_03_L_01\"),Transform=(Rotation=(X=0,Y=0,Z=0,W=1),Translation=(X=0,Y=0,Z=0),Scale3D= ...(truncated)"
    WeightMinimum=0
    Weight=&RigVMFunction_KalmanFloat(BufferSize=15, Value=&RigVMFunction_MathBoolToFloat(Value=@n8.Value).Result) {template="KalmanFilter::Execute(in Value,in BufferSize,out Result)"}.Result
   )
   > DeltaRotation v5_22 = RigVMFunctionReferenceNode(
    Rotation Order=ZYX
    child=(Type=Bone,Name="upperarm_r")
    ignore initial offset=true
    outputMinMax=(X=0,Y=1)
    parent=(Type=Bone,Name="spine_05")
    premult=1
    rotationMinMax=(X=5,Y=45)
    triggerAxis=Y
    twistAxis=(X=1,Y=0,Z=0)
   )
   > &RigUnit_ModifyTransforms(
    ItemToModify="((Item=(Type=Bone,Name=\"MSTR_coat_02_R_01\"),Transform=(Rotation=(X=-0.173,Y=0.0151,Z=-0.0858,W=0.9811),Translation=(X=0,Y=7,Z=0),Scale3D=(X=1,Y=1,Z=1))),(Item=(Type=Bone,Name=\"MSTR_coat_01_R_01\"),Transform=(Rotation=(X=-0.1736,Y=0,Z=0,W=0.9848),Translation ...(truncated)"
    WeightMinimum=0
    Weight=@"DeltaRotation v5_22".outputWeight
   )
   > DeltaRotation v5_23 = RigVMFunctionReferenceNode(
    Rotation Order=ZYX
    child=(Type=Bone,Name="upperarm_l")
    ignore initial offset=true
    outputMinMax=(X=0,Y=1)
    parent=(Type=Bone,Name="spine_05")
    premult=1
    rotationMinMax=(X=5,Y=45)
    triggerAxis=Y
    twistAxis=(X=1,Y=0,Z=0)
   )
   > &RigUnit_ModifyTransforms(
    ItemToModify="((Item=(Type=Bone,Name=\"MSTR_coat_02_L_01\"),Transform=(Rotation=(X=0.173,Y=-0.0151,Z=-0.0858,W=0.9811),Translation=(X=0,Y=7,Z=0),Scale3D=(X=1,Y=1,Z=1))),(Item=(Type=Bone,Name=\"MSTR_coat_01_L_01\"),Transform=(Rotation=(X=0.1736,Y=0,Z=0,W=0.9848),Translation= ...(truncated)"
    WeightMinimum=0
    Weight=@"DeltaRotation v5_23".outputWeight
   )
  ]}
 New Function_ContainedGraph {nodes [
   n1 = RigVMFunctionEntryNode()
   RigVMFunction_ControlFlowBranch = &RigVMFunction_ControlFlowBranch(
    ExecuteContext=@n1.ExecuteContext
    Condition=RigVMDispatchNode(A=@n1.child, B=@n1.parent) {template="DISPATCH_RigVMDispatch_CoreEquals(in A,in B,out Result)"}.Result
   )
   RigVMFunction_MathBoolToInteger = &RigVMFunction_MathBoolToInteger(Value=@n1.onlyTwist)
   RigVMFunction_MathQuaternionSwingTwist = &RigVMFunction_MathQuaternionSwingTwist(twistAxis=@n1.twistAxis, Input=RigVMVariableNode() {variable=RelativeTransform_Rotation}.Value)
   n2 = RigVMDispatchNode(
    Values=((),())
    Index=@RigVMFunction_MathBoolToInteger.Result
    Values.0=@RigVMFunction_MathQuaternionSwingTwist.Swing
    Values.1=@RigVMFunction_MathQuaternionSwingTwist.Twist
   ) {template="DISPATCH_RigVMDispatch_SelectInt32(in Index,in Values,out Result)"}
   RigVMFunction_MathQuaternionToEuler = &RigVMFunction_MathQuaternionToEuler(Value=@n2.Result, RotationOrder=@n1.Rotation Order)
   DISPATCH_RigVMDispatch_Print = RigVMDispatchNode(
    ScreenColor=(R=1,G=1,B=1,A=1)
    ScreenDuration=1.5
    Enabled=@n1.debugRotation
    ExecuteContext=@RigVMFunction_ControlFlowBranch.Completed
    Value=@RigVMFunction_MathQuaternionToEuler.Result
   ) {template="DISPATCH_RigVMDispatch_Print(in Prefix,in Value,in Enabled,in ScreenDuration,in ScreenColor)"}
   Cast = RigVMDispatchNode(Value=@n1.triggerAxis) {template="DISPATCH_RigVMDispatch_CastEnumToInt(in Value,out Result)"}
   DISPATCH_RigVMDispatch_SelectInt32 = RigVMDispatchNode(
    Values=(0,0,0)
    Index=@Cast.Result
    Values.0=@RigVMFunction_MathQuaternionToEuler.Result.X
    Values.1=@RigVMFunction_MathQuaternionToEuler.Result.Y
    Values.2=@RigVMFunction_MathQuaternionToEuler.Result.Z
   ) {template="DISPATCH_RigVMDispatch_SelectInt32(in Index,in Values,out Result)"}
   Remap = &RigVMFunction_MathDoubleRemap(
    bClamp=False
    Value=@DISPATCH_RigVMDispatch_SelectInt32.Result
    TargetMinimum=@n1.outputMinMax.X
    TargetMaximum=@n1.outputMinMax.Y
    SourceMinimum=@n1.rotationMinMax.X
    SourceMaximum=@n1.rotationMinMax.Y
   )
   Absolute = &RigVMFunction_MathDoubleAbs(Value=@Remap.Result)
   Multiply = &RigVMFunction_MathDoubleMul(
    A=&RigVMFunction_MathDoubleClamp(Value=RigVMDispatchNode(True=@Absolute.Result, Condition=@n1.absoluteValue, False=@Remap.Result) {template="DISPATCH_RigVMDispatch_If(in Condition,in True,in False,out Result)"}.Result).Result
    B=@n1.premult
   )
   > RigVMFunctionReturnNode(outputQuat=@n2.Result, outputWeight=@Multiply.Result)
   n3 = &RigVMFunction_ControlFlowBranch(Condition=@n1.ignore initial offset, ExecuteContext=@RigVMFunction_ControlFlowBranch.False)
   n4 = &RigUnit_GetRelativeTransformForItem(bChildInitial=False, bParentInitial=False, child=@n1.child, parent=@n1.parent)
   RigVMVariableNode(Value=@n4.RelativeTransform.Rotation, ExecuteContext=@n3.False) {variable=RelativeTransform_Rotation}
   RigVMVariableNode(
    Value=&RigVMFunction_MathTransformMakeRelative(
     Global=@n4.RelativeTransform
     Parent=&RigUnit_GetRelativeTransformForItem(bChildInitial=true, bParentInitial=true, child=@n1.child, parent=@n1.parent).RelativeTransform
    ).Local.Rotation
    ExecuteContext=@n3.True
   ) {variable=RelativeTransform_Rotation}
   RigVMVariableNode(
    Value=&RigVMFunction_MathTransformMakeRelative(
     Global=&RigUnit_GetTransform(Space=LocalSpace, Item=@n1.child).Transform
     Parent=&RigUnit_GetTransform(Space=LocalSpace, bInitial=true, Item=@n1.child).Transform
    ).Local.Rotation
    ExecuteContext=@RigVMFunction_ControlFlowBranch.True
   ) {variable=RelativeTransform_Rotation}
  ]}
 New Function_ContainedGraph {nodes [
   RigVMFunctionEntryNode()
   > &RigUnit_ParentConstraint(
    AdvancedSettings=(InterpolationType=Shortest,RotationOrderForFilter=ZYX)
    Child=(Type=Bone,Name="MSTR_foot_pivot_l")
    Parents=((Item=(Type=Bone,Name="calf_l"),Weight=1))
    Weight=1
   )
   > &RigUnit_ParentConstraint(
    AdvancedSettings=(InterpolationType=Shortest,RotationOrderForFilter=ZYX)
    Child=(Type=Bone,Name="MSTR_foot_pivot_r")
    Parents=((Item=(Type=Bone,Name="calf_r"),Weight=1))
    Weight=1
   )
   > /Script/ControlRig.RigUnit_PositionConstraintLocalSpaceOffset(
    Child=(Type=Bone,Name="dyn_foot_pivot_l")
    Filter=(bX=True,bY=True,bZ=True)
    Parents=((Item=(Type=Bone,Name="foot_l"),Weight=1))
    Weight=1
    bMaintainOffset=True
   ) {is=struct RigVMUnitNode}
   > &RigUnit_PositionConstraintLocalSpaceOffset(
    Child=(Type=Bone,Name="dyn_foot_pivot_r")
    Filter=(bX=True,bY=True,bZ=True)
    Parents=((Item=(Type=Bone,Name="foot_r"),Weight=1))
    Weight=1
    bMaintainOffset=True
   )
   > &RigUnit_RotationConstraintLocalSpaceOffset(
    AdvancedSettings=(InterpolationType=Average,RotationOrderForFilter=XZY)
    Child=(Type=Bone,Name="dyn_foot_pivot_l")
    Filter=(bX=True,bY=True,bZ=True)
    Parents=((Item=(Type=Bone,Name="foot_l"),Weight=1))
    Weight=1
   )
   > &RigUnit_RotationConstraintLocalSpaceOffset(
    AdvancedSettings=(InterpolationType=Average,RotationOrderForFilter=XZY)
    Child=(Type=Bone,Name="dyn_foot_pivot_r")
    Filter=(bX=True,bY=True,bZ=True)
    Parents=((Item=(Type=Bone,Name="foot_r"),Weight=1))
    Weight=1
   )
   > &RigUnit_ParentConstraint(
    AdvancedSettings=(InterpolationType=Shortest,RotationOrderForFilter=ZYX)
    Child=(Type=Bone,Name="MSTR_ball_pivot_l")
    Parents=((Item=(Type=Bone,Name="foot_l"),Weight=1))
    Weight=1
   )
   > &RigUnit_ParentConstraint(
    AdvancedSettings=(InterpolationType=Shortest,RotationOrderForFilter=ZYX)
    Child=(Type=Bone,Name="MSTR_ball_pivot_r")
    Parents=((Item=(Type=Bone,Name="foot_r"),Weight=1))
    Weight=1
   )
   > &RigUnit_PositionConstraintLocalSpaceOffset(
    Child=(Type=Bone,Name="dyn_ball_pivot_l")
    Filter=(bX=True,bY=True,bZ=True)
    Parents=((Item=(Type=Bone,Name="ball_l"),Weight=1))
    Weight=1
    bMaintainOffset=True
   )
   > &RigUnit_PositionConstraintLocalSpaceOffset(
    Child=(Type=Bone,Name="dyn_ball_pivot_r")
    Filter=(bX=True,bY=True,bZ=True)
    Parents=((Item=(Type=Bone,Name="ball_r"),Weight=1))
    Weight=1
    bMaintainOffset=True
   )
   > &RigUnit_RotationConstraintLocalSpaceOffset(
    AdvancedSettings=(InterpolationType=Average,RotationOrderForFilter=XZY)
    Child=(Type=Bone,Name="dyn_ball_pivot_l")
    Filter=(bX=True,bY=True,bZ=True)
    Parents=((Item=(Type=Bone,Name="ball_l"),Weight=1))
    Weight=1
   )
   > &RigUnit_RotationConstraintLocalSpaceOffset(
    AdvancedSettings=(InterpolationType=Average,RotationOrderForFilter=XZY)
    Child=(Type=Bone,Name="dyn_ball_pivot_r")
    Filter=(bX=True,bY=True,bZ=True)
    Parents=((Item=(Type=Bone,Name="ball_r"),Weight=1))
    Weight=1
   )
   > &RigUnit_RotationConstraintLocalSpaceOffset(
    AdvancedSettings=(InterpolationType=Average,RotationOrderForFilter=XZY)
    Child=(Type=Bone,Name="sandal_l")
    Filter=(bX=True,bY=True,bZ=True)
    Parents=((Item=(Type=Bone,Name="ball_l"),Weight=0.2),(Item=(Type=Bone,Name="foot_l"),Weight=0.8))
    Weight=1
   )
   > &RigUnit_RotationConstraintLocalSpaceOffset(
    AdvancedSettings=(InterpolationType=Average,RotationOrderForFilter=XZY)
    Child=(Type=Bone,Name="sandal_r")
    Filter=(bX=True,bY=True,bZ=True)
    Parents=((Item=(Type=Bone,Name="ball_r"),Weight=0.2),(Item=(Type=Bone,Name="foot_r"),Weight=0.8))
    Weight=1
   )
   > RigVMFunctionReturnNode()
  ]}
 New Function_ContainedGraph {nodes [
   RigVMFunctionEntryNode()
   > &RigUnit_ParentConstraint(
    AdvancedSettings=(InterpolationType=Average,RotationOrderForFilter=XZY)
    Child=(Type=Bone,Name="MSTR_sleeve_l")
    Filter=(TranslationFilter=(bX=true,bY=true,bZ=true),RotationFilter=(bX=True,bY=True,bZ=True),ScaleFilter=(bX=false,bY=false,bZ=false))
    Parents=((Item=(Type=Bone,Name="lowerarm_twist_01_l"),Weight=0.85),(Item=(Type=Bone,Name="upperarm_l"),Weight=0.15))
    Weight=1
   )
   > &RigUnit_ParentConstraint(
    AdvancedSettings=(InterpolationType=Average,RotationOrderForFilter=XZY)
    Child=(Type=Bone,Name="MSTR_sleeve_r")
    Filter=(TranslationFilter=(bX=true,bY=true,bZ=true),RotationFilter=(bX=True,bY=True,bZ=True),ScaleFilter=(bX=false,bY=false,bZ=false))
    Parents=((Item=(Type=Bone,Name="lowerarm_twist_01_r"),Weight=0.85),(Item=(Type=Bone,Name="upperarm_r"),Weight=0.15))
    Weight=1
   )
   > DeltaRotation v5 = RigVMFunctionReferenceNode(
    Rotation Order=ZYX
    child=(Type=Bone,Name="upperarm_l")
    ignore initial offset=true
    onlyTwist=false
    outputMinMax=(X=0,Y=1)
    parent=(Type=Bone,Name="spine_05")
    premult=1
    rotationMinMax=(X=15,Y=30)
    triggerAxis=Y
    twistAxis=(X=1,Y=0,Z=0)
   )
   > &RigUnit_ModifyTransforms(
    ItemToModify=((Item=(Type=Bone,Name="dyn_sleeve_l"),Transform=(Rotation=(X=0,Y=-0.1305,Z=0,W=0.9914),Translation=(X=0,Y=0,Z=0),Scale3D=(X=1,Y=1,Z=1))))
    WeightMinimum=0
    Weight=@"DeltaRotation v5".outputWeight
   )
   > DeltaRotation v5_1 = RigVMFunctionReferenceNode(
    Rotation Order=ZYX
    child=(Type=Bone,Name="upperarm_r")
    ignore initial offset=true
    onlyTwist=false
    outputMinMax=(X=0,Y=1)
    parent=(Type=Bone,Name="spine_05")
    premult=1
    rotationMinMax=(X=15,Y=30)
    triggerAxis=Y
    twistAxis=(X=1,Y=0,Z=0)
   )
   > &RigUnit_ModifyTransforms(
    ItemToModify=((Item=(Type=Bone,Name="dyn_sleeve_r"),Transform=(Rotation=(X=0,Y=-0.1305,Z=0,W=0.9914),Translation=(X=0,Y=0,Z=0),Scale3D=(X=1,Y=1,Z=1))))
    WeightMinimum=0
    Weight=@"DeltaRotation v5_1".outputWeight
   )
   > DeltaRotation v5_2 = RigVMFunctionReferenceNode(
    Rotation Order=ZYX
    child=(Type=Bone,Name="lowerarm_l")
    onlyTwist=false
    outputMinMax=(X=0,Y=1)
    parent=(Type=Bone,Name="lowerarm_l")
    premult=1
    rotationMinMax=(X=10,Y=70)
    triggerAxis=Z
    twistAxis=(X=1,Y=0,Z=0)
   )
   > &RigUnit_ModifyTransforms(
    ItemToModify="((Item=(Type=Bone,Name=\"dyn_sleeve_l\"),Transform=(Rotation=(X=0,Y=0,Z=-0.0872,W=0.9962),Translation=(X=0,Y=0,Z=0),Scale3D=(X=1,Y=1,Z=1))),(Item=(Type=Bone,Name=\"dyn_sleeve_up_l\"),Transform=(Rotation=(X=0,Y=0,Z=-0.0872,W=0.9962),Translation=(X=0,Y=0,Z=0),Sc ...(truncated)"
    WeightMinimum=0
    Weight=@"DeltaRotation v5_2".outputWeight
   )
   > DeltaRotation v5_3 = RigVMFunctionReferenceNode(
    Rotation Order=ZYX
    child=(Type=Bone,Name="lowerarm_r")
    onlyTwist=false
    outputMinMax=(X=0,Y=1)
    parent=(Type=Bone,Name="lowerarm_r")
    premult=1
    rotationMinMax=(X=10,Y=70)
    triggerAxis=Z
    twistAxis=(X=1,Y=0,Z=0)
   )
   > &RigUnit_ModifyTransforms(
    ItemToModify="((Item=(Type=Bone,Name=\"dyn_sleeve_r\"),Transform=(Rotation=(X=0,Y=0,Z=-0.0872,W=0.9962),Translation=(X=0,Y=0,Z=0),Scale3D=(X=1,Y=1,Z=1))),(Item=(Type=Bone,Name=\"dyn_sleeve_up_r\"),Transform=(Rotation=(X=0,Y=0,Z=-0.0872,W=0.9962),Translation=(X=0,Y=0,Z=0),Sc ...(truncated)"
    WeightMinimum=0
    Weight=@"DeltaRotation v5_3".outputWeight
   )
   > DeltaRotation v5_4 = RigVMFunctionReferenceNode(
    Rotation Order=ZYX
    child=(Type=Bone,Name="upperarm_l")
    ignore initial offset=true
    onlyTwist=false
    outputMinMax=(X=0,Y=1)
    parent=(Type=Bone,Name="spine_05")
    premult=1
    rotationMinMax=(X=0,Y=-70)
    triggerAxis=Y
    twistAxis=(X=1,Y=0,Z=0)
   )
   > &RigUnit_ModifyTransforms(
    ItemToModify="((Item=(Type=Bone,Name=\"dyn_sleeve_up_l\"),Transform=(Rotation=(X=0,Y=0.0872,Z=0,W=0.9962),Translation=(X=2.0953,Y=0,Z=-1.0804),Scale3D=(X=1,Y=1,Z=1))),(Item=(Type=Bone,Name=\"dyn_sleeve_l\"),Transform=(Rotation=(X=0,Y=0.0872,Z=0,W=0.9962),Translation=(X=0,Y= ...(truncated)"
    WeightMinimum=0
    Weight=@"DeltaRotation v5_4".outputWeight
   )
   > DeltaRotation v5_5 = RigVMFunctionReferenceNode(
    Rotation Order=ZYX
    child=(Type=Bone,Name="upperarm_r")
    ignore initial offset=true
    onlyTwist=false
    outputMinMax=(X=0,Y=1)
    parent=(Type=Bone,Name="spine_05")
    premult=1
    rotationMinMax=(X=0,Y=-70)
    triggerAxis=Y
    twistAxis=(X=1,Y=0,Z=0)
   )
   > &RigUnit_ModifyTransforms(
    ItemToModify="((Item=(Type=Bone,Name=\"dyn_sleeve_up_r\"),Transform=(Rotation=(X=0,Y=0.0872,Z=0,W=0.9962),Translation=(X=-2.0953,Y=0,Z=1.0804),Scale3D=(X=1,Y=1,Z=1))),(Item=(Type=Bone,Name=\"dyn_sleeve_r\"),Transform=(Rotation=(X=0,Y=0.0872,Z=0,W=0.9962),Translation=(X=0,Y= ...(truncated)"
    WeightMinimum=0
    Weight=@"DeltaRotation v5_5".outputWeight
   )
   > DeltaRotation v5_6 = RigVMFunctionReferenceNode(
    Rotation Order=ZYX
    child=(Type=Bone,Name="upperarm_l")
    ignore initial offset=true
    onlyTwist=false
    outputMinMax=(X=0,Y=1)
    parent=(Type=Bone,Name="spine_05")
    premult=0.5
    rotationMinMax=(X=0,Y=70)
    triggerAxis=Z
    twistAxis=(X=1,Y=0,Z=0)
   )
   > &RigUnit_ModifyTransforms(
    ItemToModify="((Item=(Type=Bone,Name=\"dyn_sleeve_up_l\"),Transform=(Rotation=(X=-0.374,Y=0.0295,Z=0.0457,W=0.9258),Translation=(X=0,Y=0,Z=0),Scale3D=(X=1,Y=1,Z=1))),(Item=(Type=Bone,Name=\"dyn_sleeve_l\"),Transform=(Rotation=(X=-0.1288,Y=-0.0464,Z=-0.2165,W=0.9666),Transla ...(truncated)"
    WeightMinimum=0
    Weight=@"DeltaRotation v5_6".outputWeight
   )
   > DeltaRotation v5_7 = RigVMFunctionReferenceNode(
    Rotation Order=ZYX
    child=(Type=Bone,Name="upperarm_r")
    onlyTwist=false
    outputMinMax=(X=0,Y=1)
    parent=(Type=Bone,Name="spine_05")
    premult=0.5
    rotationMinMax=(X=0,Y=70)
    triggerAxis=Z
    twistAxis=(X=1,Y=0,Z=0)
   )
   > &RigUnit_ModifyTransforms(
    ItemToModify="((Item=(Type=Bone,Name=\"dyn_sleeve_up_r\"),Transform=(Rotation=(X=0.5306,Y=0.0141,Z=-0.1101,W=-0.8403),Translation=(X=0,Y=0,Z=0),Scale3D=(X=1,Y=1,Z=1))),(Item=(Type=Bone,Name=\"dyn_sleeve_r\"),Transform=(Rotation=(X=-0.2216,Y=0.0054,Z=-0.403,W=0.8879),Transla ...(truncated)"
    WeightMinimum=0
    Weight=@"DeltaRotation v5_7".outputWeight
   )
   > RigVMFunctionReturnNode()
  ]}
 New Function_ContainedGraph {nodes [
   DeltaRotation v5 = RigVMFunctionReferenceNode(
    ExecuteContext=()
    Rotation Order=ZYX
    child=(Type=Bone,Name="upperarm_l")
    onlyTwist=false
    outputMinMax=(X=0,Y=1)
    parent=(Type=Null,Name="upperarm_l_null")
    premult=1
    rotationMinMax=(X=15,Y=35)
    triggerAxis=Y
    twistAxis=(X=1,Y=0,Z=0)
   )
   > &RigUnit_ModifyTransforms(
    ItemToModify=((Item=(Type=Bone,Name="dyn_sleeve_l"),Transform=(Rotation=(X=0,Y=0.0669,Z=0,W=-0.9978),Translation=(X=-0.5124,Y=0,Z=7.6428),Scale3D=(X=1,Y=1,Z=1))))
    WeightMinimum=0
    Weight=@"DeltaRotation v5".outputWeight
   )
   RigVMFunctionEntryNode()
   > RigVMFunctionReturnNode()
  ]}
 RigVMFunctionLibrary {nodes [
   DeltaRotation v5 = RigVMCollapseNode(
    ExecuteContext=()
    Rotation Order=ZYX
    child=(Type=Bone,Name="None")
    outputMinMax=(X=0,Y=1)
    parent=(Type=Bone,Name="None")
    premult=1
    rotationMinMax=(X=0,Y=0)
    twistAxis=(X=1,Y=0,Z=0)
   )
   coat = RigVMCollapseNode(ExecuteContext=())
   collar = RigVMCollapseNode(ExecuteContext=())
   shoes = RigVMCollapseNode(ExecuteContext=())
   sleeves = RigVMCollapseNode(ExecuteContext=())
  ]}
 RigVMFunction_Sequence_2_ContainedGraph {nodes [
   n1 = RigVMFunctionEntryNode()
   n2 = &RigVMFunction_Sequence(ExecuteContext=@n1.ExecuteContext)
   n3 = &RigVMFunction_Sequence(ExecuteContext=@n2.B)
   n4 = &RigVMFunction_Sequence(ExecuteContext=@n3.B)
   n5 = &RigVMFunction_Sequence(ExecuteContext=@n4.B)
   n6 = &RigVMFunction_Sequence(ExecuteContext=@n5.B)
   RigVMFunctionReturnNode(A=@n2.A, B=@n3.A, C=@n4.A, D=@n5.A, E=@n6.A, F=@n6.B)
  ]}
 RigVMModel {nodes [
   RigUnit_BeginExecution = &RigUnit_BeginExecution()
   /Script/FortniteGame.RigUnit_FortDeformRig(
    Bool=(bUpperArm=true,bElbow=true,bWrist=true,bGroin=true,bGlute=False,bKnee=true,bPec=false,bBicep=true,bLat=true)
    Curve="(bLUpperArmUpAlpha=False,bRUpperArmUpAlpha=False,bLUpperArmFrAlpha=False,bRUpperArmFrAlpha=False,bLUpperArmBkAlpha=False,bRUpperArmBkAlpha=False,bLWristUpAlpha=False,bRWristUpAlpha=False,bLWristDnAlpha=False,bRWristDnAlpha=False,bLKneeAlpha=False,bRKneeAlp ...(truncated)"
    Debug=(bUpperArmDebug=False,bWristDebug=False,bLatDebug=False,bGroinGluteDebug=False)
    Value="(UpperArmUp=(Rotation=(X=0.0,Y=-0.0,Z=0.0,W=1.0),Translation=(X=0.0,Y=0.0,Z=1),Scale3D=(X=2,Y=1.0,Z=1.0)),UpperArmFwdFr=(Rotation=(X=0.0,Y=-0.0,Z=0.0,W=1.0),Translation=(X=0.0,Y=10.0,Z=0.0),Scale3D=(X=0.50,Y=1.0,Z=1.0)),UpperArmFwdBk=(Rotation=(X=0.0,Y=-0. ...(truncated)"
    ExecutePin=@RigUnit_BeginExecution.ExecutePin
   ) {is=struct RigVMUnitNode}
   > shoes = RigVMFunctionReferenceNode()
   > sleeves = RigVMFunctionReferenceNode()
   > coat = RigVMFunctionReferenceNode()
   RigVMFunction_MathBoolToFloat = &RigVMFunction_MathBoolToFloat(Value=RigVMVariableNode() {variable=crouching}.Value)
   KalmanFilter = &RigVMFunction_KalmanFloat(BufferSize=5, Value=@RigVMFunction_MathBoolToFloat.Result) {template="KalmanFilter::Execute(in Value,in BufferSize,out Result)"}
   > &RigUnit_ModifyTransforms(
    ItemToModify="((Item=(Type=Bone,Name=\"MSTR_ribbon_B_r\"),Transform=(Rotation=(X=0,Y=0,Z=0,W=1),Translation=(X=0,Y=0,Z=0),Scale3D=(X=0.95,Y=1,Z=1))),(Item=(Type=Bone,Name=\"MSTR_ribbon_B_l\"),Transform=(Rotation=(X=0,Y=0,Z=0.0436,W=0.999),Translation=(X=0,Y=0,Z=0),Scale3D=( ...(truncated)"
    WeightMinimum=0
    Weight=@KalmanFilter.Result
   )
  ]}
]
hierarchy [
 root {loc=<0 0 0>, rot=<0 0 0 1>}
 pelvis {parent=root, loc=<0 -0.7071 93.1191>, rot=<0.0127 0.707 -0.0127 -0.707>}
 spine_01 {parent=pelvis, loc=<6.7494 -0.6388 0>, rot=<0 0 0.034 -0.9994>}
 spine_02 {parent=spine_01, loc=<7.2518 -0.035 0>, rot=<0 0 0.0238 -0.9997>}
 spine_03 {parent=spine_02, loc=<6.0104 0 0>, rot=<0 0 0.0149 -0.9999>}
 spine_04 {parent=spine_03, loc=<8.1024 0 0>, rot=<0 0 -0.0132 -0.9999>}
 spine_05 {parent=spine_04, loc=<10.0146 0 0>, rot=<0 0 -0.0015 -1>}
 neck_01 {parent=spine_05, loc=<11.3138 0.1529 0>, rot=<0 0 -0.2003 -0.9797>}
 neck_02 {parent=neck_01, loc=<4.5627 0 0>, rot=<0 0 0.0504 -0.9987>}
 head {parent=neck_02, loc=<5.3914 0 0>, rot=<0 0 0.115 -0.9934>}
 MSTR_collar_bk {parent=neck_02, loc=<0.006 -3.8614 0>, rot=<0 0 -0.0589 -0.9983>}
 dyn_collar_bk {parent=MSTR_collar_bk, loc=<0 0 0>, rot=<0 0 0 1>}
 MSTR_collar_l {parent=neck_02, loc=<-2.0308 -0.28 -4.921>, rot=<-0.6869 -0.1677 -0.1677 -0.6869>}
 dyn_collar_l {parent=MSTR_collar_l, loc=<0 0 0>, rot=<0 0 0 1>}
 MSTR_collar_r {parent=neck_02, loc=<-2.0305 -0.2801 4.921>, rot=<0.6869 0.1677 -0.1677 -0.6869>}
 dyn_collar_r {parent=MSTR_collar_r, loc=<0 0 0>, rot=<0 0 0 1>}
 attach_cape {parent=spine_05, loc=<9.017 -11.714 0>, rot=<0.0616 -0.7044 -0.0616 -0.7044>}
 cape {parent=attach_cape, type=Null, loc=<0 0.7597 0.2515>, rot=<0.0066 0 0 1>}
 attach_backpack {parent=spine_05, loc=<-0.123 -13.473 0>, rot=<-0.0093 -0.707 0.0093 -0.707>}
 Backpack_BR {parent=attach_backpack, type=Null, loc=<0 0.5146 -0.1312>, rot=<-0.0886 0 0 0.9961>}
 clavicle_l {parent=spine_05, loc=<4.3575 -0.1466 -2.3223>, rot=<0.0465 -0.7289 -0.009 -0.6829>}
 upperarm_l {parent=clavicle_l, loc=<15.5111 0 0>, rot=<0.0226 -0.3234 -0.0088 -0.9459>}
 lowerarm_l {parent=upperarm_l, loc=<26.6514 0 0>, rot=<0.0391 -0.0099 -0.2453 -0.9686>}
 hand_l {parent=lowerarm_l, loc=<22.3491 0 0>, rot=<-0.7081 -0.0507 0.0975 -0.6975>}
 thumb_01_l {parent=hand_l, loc=<1.3118 -0.3191 -2.459>, rot=<0.6546 -0.4042 0.0664 -0.6354>}
 thumb_02_l {parent=thumb_01_l, loc=<5.9336 -0.0448 0.0401>, rot=<-0.0037 0.0213 0.298 -0.9543>}
 thumb_03_l {parent=thumb_02_l, loc=<2.2223 0.3682 -0.0907>, rot=<0.0118 -0.013 0.0066 -0.9998>}
 index_metacarpal_l {parent=hand_l, loc=<1.6535 0.0183 -2.96>, rot=<0.0158 -0.0124 -0.0173 -0.9996>}
 index_01_l {parent=index_metacarpal_l, loc=<8.3008 0.0357 -0.0524>, rot=<-0.0044 0.0485 0.0924 -0.9945>}
 index_02_l {parent=index_01_l, loc=<3.3587 -0.3004 0.0452>, rot=<-0.002 0.0027 0.3225 -0.9466>}
 index_03_l {parent=index_02_l, loc=<2.5578 -0.123 -0.0252>, rot=<-0.002 0.0067 0.2558 -0.9667>}
 middle_metacarpal_l {parent=hand_l, loc=<2.1856 -0.1725 -0.5984>, rot=<0.006 0.0168 -0.0329 -0.9993>}
 middle_01_l {parent=middle_metacarpal_l, loc=<7.3275 -0.2447 0.0129>, rot=<-0.0208 0.0042 0.0858 -0.9961>}
 middle_02_l {parent=middle_01_l, loc=<4.0123 0.1401 0.1175>, rot=<-0.0218 0.0258 0.3512 -0.9357>}
 middle_03_l {parent=middle_02_l, loc=<2.6077 0.0534 0.1373>, rot=<0.0082 0.0188 0.1708 -0.9851>}
 ring_metacarpal_l {parent=hand_l, loc=<2.2074 -0.3625 1.3638>, rot=<-0.0284 0.0722 -0.0458 -0.9959>}
 ring_01_l {parent=ring_metacarpal_l, loc=<7.0295 -0.6197 -0.1446>, rot=<-0.0489 -0.0252 0.1892 -0.9804>}
 ring_02_l {parent=ring_01_l, loc=<3.6174 0.4647 -0.1027>, rot=<0 0 0.2806 -0.9598>}
 ring_03_l {parent=ring_02_l, loc=<2.4152 0.114 0.0199>, rot=<0 0 0.1189 -0.9929>}
 pinky_metacarpal_l {parent=hand_l, loc=<1.7616 -0.9753 3.3814>, rot=<-0.0397 0.0814 -0.069 -0.9935>}
 pinky_01_l {parent=pinky_metacarpal_l, loc=<7.1709 -0.4831 -0.1094>, rot=<-0.0654 -0.0035 0.2242 -0.9723>}
 pinky_02_l {parent=pinky_01_l, loc=<2.8274 0.0893 -0.1043>, rot=<0.0037 0.0011 0.2854 -0.9584>}
 pinky_03_l {parent=pinky_02_l, loc=<2.2046 0.0307 -0.0078>, rot=<-0.0002 0.0098 0.1445 -0.9895>}
 deform_wrist_dn_base_l {parent=hand_l, loc=<0 -2.2342 0>, rot=<0 0 0 -1>}
 deform_wrist_dn_l {parent=deform_wrist_dn_base_l, loc=<0 0 0>, rot=<0 0 0 1>}
 deform_wrist_up_base_l {parent=hand_l, loc=<0 2.2682 0>, rot=<0 0 0 -1>}
 deform_wrist_up_l {parent=deform_wrist_up_base_l, loc=<0 0 0>, rot=<0 0 0 1>}
 lowerarm_twist_02_l {parent=lowerarm_l, loc=<9.6199 0 0>, rot=<0 0 0 -1>}
 lowerarm_twist_01_l {parent=lowerarm_l, loc=<17.8549 0 0>, rot=<0 0 0 -1>}
 deform_elbow_in_base_l {parent=lowerarm_l, loc=<0.2112 5.259 0.8274>, rot=<0 0 0 -1>}
 deform_elbow_in_l {parent=deform_elbow_in_base_l, loc=<0 0 0>, rot=<0 0 0 1>}
 deform_elbow_out_base_l {parent=lowerarm_l, loc=<0 -4 0>, rot=<0 0 0 -1>}
 deform_elbow_out_l {parent=deform_elbow_out_base_l, loc=<0 0 0>, rot=<0 0 0 1>}
 MSTR_sleeve_l {parent=lowerarm_l, loc=<0.9475 2.5443 -6.7368>, rot=<-0.0798 -0.0784 0.0989 -0.9888>}
 dyn_sleeve_l {parent=MSTR_sleeve_l, loc=<0 0 0>, rot=<0 0 0 1>}
 MSTR_sleeve_up_l {parent=lowerarm_l, loc=<-0.058 0.6211 4.4624>, rot=<-0.0889 0.035 0.1066 -0.9897>}
 dyn_sleeve_up_l {parent=MSTR_sleeve_up_l, loc=<0 0 0>, rot=<0 0 0 1>}
 upperarm_twist_01_l {parent=upperarm_l, loc=<0.0041 0 0>, rot=<0 0 0 -1>}
 deform_upperarm_bk_base_l {parent=upperarm_twist_01_l, loc=<1.0813 -7.2193 -1.9453>, rot=<0 0 0 1>}
 deform_upperarm_bk_l {parent=deform_upperarm_bk_base_l, loc=<0 0 0>, rot=<0 0 0 1>}
 deform_upperarm_fr_base_l {parent=upperarm_twist_01_l, loc=<-0.3691 7.0506 -2.765>, rot=<0 0 0 1>}
 deform_upperarm_fr_l {parent=deform_upperarm_fr_base_l, loc=<0 0 0>, rot=<0 0 0 1>}
 deform_upperarm_up_base_l {parent=upperarm_twist_01_l, loc=<-3.2216 -1.1391 3.3311>, rot=<0 0 0 1>}
 deform_upperarm_up_l {parent=deform_upperarm_up_base_l, loc=<0 0 0>, rot=<0 0 0 1>}
 upperarm_twist_02_l {parent=upperarm_l, loc=<13.7941 0 0>, rot=<0 0 0 -1>}
 clavicle_r {parent=spine_05, loc=<4.3575 -0.1466 2.3223>, rot=<0.7289 0.0465 -0.6829 0.009>}
 upperarm_r {parent=clavicle_r, loc=<-15.5111 0 0>, rot=<0.0226 -0.3234 -0.0088 -0.9459>}
 lowerarm_r {parent=upperarm_r, loc=<-26.6514 0 0>, rot=<0.0391 -0.0099 -0.2453 -0.9686>}
 hand_r {parent=lowerarm_r, loc=<-22.3491 0 0>, rot=<-0.7081 -0.0507 0.0975 -0.6975>}
 thumb_01_r {parent=hand_r, loc=<-1.3118 0.3191 2.459>, rot=<0.6546 -0.4042 0.0664 -0.6354>}
 thumb_02_r {parent=thumb_01_r, loc=<-5.9336 0.0448 -0.0401>, rot=<-0.0037 0.0213 0.298 -0.9543>}
 thumb_03_r {parent=thumb_02_r, loc=<-2.2223 -0.3682 0.0907>, rot=<0.0118 -0.013 0.0066 -0.9998>}
 index_metacarpal_r {parent=hand_r, loc=<-1.6535 -0.0183 2.96>, rot=<0.0158 -0.0124 -0.0173 -0.9996>}
 index_01_r {parent=index_metacarpal_r, loc=<-8.3008 -0.0357 0.0524>, rot=<-0.0044 0.0485 0.0924 -0.9945>}
 index_02_r {parent=index_01_r, loc=<-3.3587 0.3004 -0.0452>, rot=<-0.002 0.0027 0.3225 -0.9466>}
 index_03_r {parent=index_02_r, loc=<-2.5578 0.123 0.0252>, rot=<-0.002 0.0067 0.2558 -0.9667>}
 middle_metacarpal_r {parent=hand_r, loc=<-2.1856 0.1725 0.5984>, rot=<0.006 0.0168 -0.0329 -0.9993>}
 middle_01_r {parent=middle_metacarpal_r, loc=<-7.3275 0.2447 -0.0129>, rot=<-0.0208 0.0042 0.0858 -0.9961>}
 middle_02_r {parent=middle_01_r, loc=<-4.0123 -0.1401 -0.1175>, rot=<-0.0218 0.0258 0.3512 -0.9357>}
 middle_03_r {parent=middle_02_r, loc=<-2.6077 -0.0534 -0.1373>, rot=<0.0082 0.0188 0.1708 -0.9851>}
 ring_metacarpal_r {parent=hand_r, loc=<-2.2074 0.3625 -1.3638>, rot=<-0.0284 0.0722 -0.0458 -0.9959>}
 ring_01_r {parent=ring_metacarpal_r, loc=<-7.0295 0.6197 0.1446>, rot=<-0.0489 -0.0252 0.1892 -0.9804>}
 ring_02_r {parent=ring_01_r, loc=<-3.6174 -0.4647 0.1027>, rot=<0 0 0.2806 -0.9598>}
 ring_03_r {parent=ring_02_r, loc=<-2.4152 -0.114 -0.0199>, rot=<0 0 0.1189 -0.9929>}
 pinky_metacarpal_r {parent=hand_r, loc=<-1.7616 0.9753 -3.3814>, rot=<-0.0397 0.0814 -0.069 -0.9935>}
 pinky_01_r {parent=pinky_metacarpal_r, loc=<-7.1709 0.4831 0.1094>, rot=<-0.0654 -0.0035 0.2242 -0.9723>}
 pinky_02_r {parent=pinky_01_r, loc=<-2.8274 -0.0893 0.1043>, rot=<0.0037 0.0011 0.2854 -0.9584>}
 pinky_03_r {parent=pinky_02_r, loc=<-2.2046 -0.0307 0.0078>, rot=<-0.0002 0.0098 0.1445 -0.9895>}
 deform_wrist_dn_base_r {parent=hand_r, loc=<-0.0004 2.2345 -0.0001>, rot=<0 0 0 -1>}
 deform_wrist_dn_r {parent=deform_wrist_dn_base_r, loc=<0 0 0>, rot=<0 0 0 1>}
 deform_wrist_up_base_r {parent=hand_r, loc=<0.0003 -2.2684 0.0001>, rot=<0 0 0 -1>}
 deform_wrist_up_r {parent=deform_wrist_up_base_r, loc=<0 0 0>, rot=<0 0 0 1>}
 lowerarm_twist_02_r {parent=lowerarm_r, loc=<-9.6199 0 0>, rot=<0 0 0 -1>}
 lowerarm_twist_01_r {parent=lowerarm_r, loc=<-17.8549 0 0>, rot=<0 0 0 -1>}
 deform_elbow_in_base_r {parent=lowerarm_r, loc=<-0.2111 -5.2591 -0.8275>, rot=<0 0 0 -1>}
 deform_elbow_in_r {parent=deform_elbow_in_base_r, loc=<-0.0001 0 0.0002>, rot=<0 0 0 1>}
 deform_elbow_out_base_r {parent=lowerarm_r, loc=<0 4 -0.0001>, rot=<0 0 0 -1>}
 deform_elbow_out_r {parent=deform_elbow_out_base_r, loc=<0 0 0>, rot=<0 0 0 1>}
 MSTR_sleeve_r {parent=lowerarm_r, loc=<-0.9474 -2.5443 6.7368>, rot=<-0.0798 -0.0784 0.0989 -0.9888>}
 dyn_sleeve_r {parent=MSTR_sleeve_r, loc=<0 0 0>, rot=<0 0 0 1>}
 MSTR_sleeve_up_r {parent=lowerarm_r, loc=<0.0581 -0.6211 -4.4625>, rot=<-0.0889 0.035 0.1066 -0.9897>}
 dyn_sleeve_up_r {parent=MSTR_sleeve_up_r, loc=<0 0 0>, rot=<0 0 0 1>}
 upperarm_twist_01_r {parent=upperarm_r, loc=<-0.0041 0 0>, rot=<0 0 0 -1>}
 deform_upperarm_bk_base_r {parent=upperarm_twist_01_r, loc=<-1.0814 7.2192 1.9455>, rot=<0 0 0 1>}
 deform_upperarm_bk_r {parent=deform_upperarm_bk_base_r, loc=<0 0 0>, rot=<0 0 0 1>}
 deform_upperarm_fr_base_r {parent=upperarm_twist_01_r, loc=<0.3689 -7.0506 2.7652>, rot=<0 0 0 1>}
 deform_upperarm_fr_r {parent=deform_upperarm_fr_base_r, loc=<0 0 0>, rot=<0 0 0 1>}
 deform_upperarm_up_base_r {parent=upperarm_twist_01_r, loc=<3.2214 1.1391 -3.3308>, rot=<0 0 0 1>}
 deform_upperarm_up_r {parent=deform_upperarm_up_base_r, loc=<0 0 0>, rot=<0 0 0 -1>}
 upperarm_twist_02_r {parent=upperarm_r, loc=<-13.7941 0 0>, rot=<0 0 0 -1>}
 backpack_collision {parent=spine_05, loc=<-0.123 -13.473 0>, rot=<-0.0093 -0.707 0.0093 -0.707>}
 deform_pec_aim_base_l {parent=spine_04, loc=<7.8732 12.6388 -2.4123>, rot=<0.0102 -0.4602 0.3979 -0.7936>}
 deform_pec_aim_l {parent=deform_pec_aim_base_l, loc=<0 0 0>, rot=<0 0 0 1>}
 deform_pec_base_l {parent=deform_pec_aim_l, loc=<10.766 0 0>, rot=<0 0 0 1>}
 deform_pec_l {parent=deform_pec_base_l, loc=<0.055 0 0>, rot=<0 0 0 1>}
 deform_pec_aim_base_r {parent=spine_04, loc=<7.8727 12.6388 2.4123>, rot=<-0.4602 -0.0102 0.7936 0.3979>}
 deform_pec_aim_r {parent=deform_pec_aim_base_r, loc=<0 0 0>, rot=<0 0 0 1>}
 deform_pec_base_r {parent=deform_pec_aim_r, loc=<-10.7662 -0.0006 -0.0007>, rot=<0 0 0 1>}
 deform_pec_r {parent=deform_pec_base_r, loc=<0 0 0>, rot=<0 0 0 1>}
 MSTR_coat_01_L_01 {parent=spine_04, loc=<0.4542 8.0362 -10.1532>, rot=<-0.0685 0.1868 0.9797 0.0232>}
 dyn_coat_01_L_01 {parent=MSTR_coat_01_L_01, loc=<0 0 0>, rot=<0 0 0 1>}
 dyn_coat_01_L_02 {parent=dyn_coat_01_L_01, loc=<10.792 -0.0053 0>, rot=<0 0 0 1>}
 dyn_coat_01_L_03 {parent=dyn_coat_01_L_02, loc=<18.5904 0 0>, rot=<0 0 0 1>}
 dyn_coat_01_L_04 {parent=dyn_coat_01_L_03, loc=<19.0103 0 0>, rot=<0 0 0 1>}
 MSTR_coat_01_R_01 {parent=spine_04, loc=<0.4543 8.0362 10.1532>, rot=<0.0685 -0.1868 0.9797 0.0232>}
 dyn_coat_01_R_01 {parent=MSTR_coat_01_R_01, loc=<0 0 0>, rot=<0 0 0 1>}
 dyn_coat_01_R_02 {parent=dyn_coat_01_R_01, loc=<10.7923 -0.0053 0>, rot=<0 0 0 1>}
 dyn_coat_01_R_03 {parent=dyn_coat_01_R_02, loc=<18.5901 -0.0001 0>, rot=<0 0 0 1>}
 dyn_coat_01_R_04 {parent=dyn_coat_01_R_03, loc=<19.0103 0 0>, rot=<0 0 0 1>}
 MSTR_coat_B_01 {parent=spine_03, loc=<1.7788 -13.9995 0>, rot=<0.0961 -0.9954 0 0>}
 dyn_coat_B_01 {parent=MSTR_coat_B_01, loc=<0 0 0>, rot=<0 0 0 1>}
 dyn_coat_B_02 {parent=dyn_coat_B_01, loc=<17.1336 0 0>, rot=<0 0 0 1>}
 dyn_coat_B_03 {parent=dyn_coat_B_02, loc=<16.2546 0 0>, rot=<0 0 0 1>}
 MSTR_coat_03_L_01 {parent=spine_03, loc=<1.4419 -12.2141 -12.474>, rot=<0.0808 -0.9887 -0.1214 -0.0362>}
 dyn_coat_03_L_01 {parent=MSTR_coat_03_L_01, loc=<0 0 0>, rot=<0 0 0 1>}
 dyn_coat_03_L_02 {parent=dyn_coat_03_L_01, loc=<17.3086 0 0>, rot=<0 0 0 1>}
 dyn_coat_03_L_03 {parent=dyn_coat_03_L_02, loc=<16.4296 0 0>, rot=<0 0 0 1>}
 MSTR_coat_02_L_01 {parent=spine_03, loc=<-0.0429 -1.6475 -18.3911>, rot=<-0.0867 0.7018 0.7068 0.0221>}
 dyn_coat_02_L_01 {parent=MSTR_coat_02_L_01, loc=<0 0 0>, rot=<0 0 0 1>}
 dyn_coat_02_L_02 {parent=dyn_coat_02_L_01, loc=<17.6512 0 0>, rot=<0 0 0 1>}
 dyn_coat_02_L_03 {parent=dyn_coat_02_L_02, loc=<18.7727 0 0>, rot=<0 0 0 1>}
 MSTR_coat_02_R_01 {parent=spine_03, loc=<-0.0432 -1.6475 18.3911>, rot=<0.0867 -0.7018 0.7068 0.0221>}
 dyn_coat_02_R_01 {parent=MSTR_coat_02_R_01, loc=<0 0 0>, rot=<0 0 0 1>}
 dyn_coat_02_R_02 {parent=dyn_coat_02_R_01, loc=<17.651 -0.0001 0>, rot=<0 0 0 -1>}
 dyn_coat_02_R_03 {parent=dyn_coat_02_R_02, loc=<18.7727 0 0>, rot=<0 0 0 -1>}
 MSTR_coat_03_R_01 {parent=spine_03, loc=<1.2128 -12.2389 12.4742>, rot=<0.0808 -0.9887 0.1214 0.0362>}
 dyn_coat_03_R_01 {parent=MSTR_coat_03_R_01, loc=<0 0 0>, rot=<0 0 0 1>}
 dyn_coat_03_R_02 {parent=dyn_coat_03_R_01, loc=<17.3081 -0.0001 0>, rot=<0 0 0 1>}
 dyn_coat_03_R_03 {parent=dyn_coat_03_R_02, loc=<16.4296 0.0001 0>, rot=<0 0 0 1>}
 MSTR_ribbon {parent=spine_01, loc=<4.5358 8.9833 6.7946>, rot=<-0.0934 -0.9956 0.0004 0.0057>}
 MSTR_ribbon_A_l {parent=MSTR_ribbon, loc=<0.538 0.213 0.8447>, rot=<0.0427 0.2075 0.0147 -0.9772>}
 dyn_ribbon_A_l {parent=MSTR_ribbon_A_l, loc=<0 0 0>, rot=<0 0 0 1>}
 MSTR_ribbon_A_r {parent=MSTR_ribbon, loc=<0.5141 -0.0004 -0.5767>, rot=<-0.0384 -0.1913 0.0221 -0.9805>}
 dyn_ribbon_A_r {parent=MSTR_ribbon_A_r, loc=<0 0 0>, rot=<0 0 0 1>}
 MSTR_ribbon_B_l {parent=MSTR_ribbon, loc=<1.5183 0.0348 0.6493>, rot=<0.0014 0.0101 0.0495 -0.9987>}
 dyn_ribbon_B_l {parent=MSTR_ribbon_B_l, loc=<0 0 0>, rot=<0 0 0 1>}
 MSTR_ribbon_B_r {parent=MSTR_ribbon, loc=<1.5022 0.0544 -0.4712>, rot=<-0.0047 -0.0266 0.0539 -0.9982>}
 dyn_ribbon_B_r {parent=MSTR_ribbon_B_r, loc=<0 0 0>, rot=<0 0 0 1>}
 thigh_l {parent=pelvis, loc=<-2.5693 -0.0109 -7.7501>, rot=<0.0012 0.0572 0.0155 -0.9982>}
 calf_l {parent=thigh_l, loc=<-39.0484 0 0>, rot=<0.0006 -0.0161 -0.0391 -0.9991>}
 foot_l {parent=calf_l, loc=<-42.561 0.1077 0.4292>, rot=<0.0418 -0.0455 0.0343 -0.9975>}
 ball_l {parent=foot_l, loc=<-6.1599 13.8264 0.6806>, rot=<0.0001 -0.0001 -0.7063 -0.7079>}
 MSTR_ball_pivot_l {parent=ball_l, loc=<-3.7117 -4.446 -1.0718>, rot=<0 0 0 1>}
 dyn_ball_pivot_l {parent=MSTR_ball_pivot_l, loc=<0 0 0>, rot=<0 0 0 1>}
 MSTR_foot_pivot_l {parent=foot_l, loc=<3.3581 -0.0411 0.1892>, rot=<0 0 0 1>}
 dyn_foot_pivot_l {parent=MSTR_foot_pivot_l, loc=<0 0 0>, rot=<0 0 0 1>}
 sandal_l {parent=foot_l, loc=<-4.6072 -0.2966 0.683>, rot=<0.0001 -0.0001 -0.7063 -0.7079>}
 calf_twist_02_l {parent=calf_l, loc=<-18.8064 -0.0967 0.0397>, rot=<0.0237 0.0095 -0.0035 -0.9997>}
 calf_twist_01_l {parent=calf_l, loc=<-35.813 -0.0967 0.0397>, rot=<0.0237 0.0095 -0.0035 -0.9997>}
 deform_knee_in_base_l {parent=calf_l, loc=<1.298 -5.2957 -0.0496>, rot=<0 0 0 -1>}
 deform_knee_in_l {parent=deform_knee_in_base_l, loc=<0 0 0>, rot=<0 0 0 1>}
 deform_knee_out_base_l {parent=calf_l, loc=<0 4 0>, rot=<0 0 0 -1>}
 deform_knee_out_l {parent=deform_knee_out_base_l, loc=<0 0 0>, rot=<0 0 0 1>}
 thigh_twist_01_l {parent=thigh_l, loc=<-21.6303 0 0>, rot=<0.0013 0 0 -1>}
 deform_groin_base_l {parent=thigh_l, loc=<-2.1908 8.1142 -1.7442>, rot=<0 0 0 -1>}
 deform_groin_l {parent=deform_groin_base_l, loc=<0 0 0>, rot=<0 0 0 1>}
 thigh_r {parent=pelvis, loc=<-2.5693 -0.0109 7.7501>, rot=<0.0572 -0.0012 0.9982 0.0155>}
 calf_r {parent=thigh_r, loc=<39.0484 0 0>, rot=<0.0006 -0.0161 -0.0391 -0.9991>}
 foot_r {parent=calf_r, loc=<42.561 -0.1077 -0.4292>, rot=<0.0418 -0.0455 0.0343 -0.9975>}
 ball_r {parent=foot_r, loc=<6.1599 -13.8264 -0.6806>, rot=<0.0001 -0.0001 -0.7063 -0.7079>}
 MSTR_ball_pivot_r {parent=ball_r, loc=<3.7117 4.446 1.0718>, rot=<0 0 0 1>}
 dyn_ball_pivot_r {parent=MSTR_ball_pivot_r, loc=<0 0 0>, rot=<0 0 0 1>}
 MSTR_foot_pivot_r {parent=foot_r, loc=<-3.358 0.0411 -0.1892>, rot=<0 0 0 1>}
 dyn_foot_pivot_r {parent=MSTR_foot_pivot_r, loc=<0 0 0>, rot=<0 0 0 1>}
 sandal_r {parent=foot_r, loc=<4.6072 0.2966 -0.683>, rot=<0.0001 -0.0001 -0.7063 -0.7079>}
 calf_twist_02_r {parent=calf_r, loc=<18.8064 0.0967 -0.0397>, rot=<0.0237 0.0095 -0.0035 -0.9997>}
 calf_twist_01_r {parent=calf_r, loc=<35.813 0.0967 -0.0397>, rot=<0.0237 0.0095 -0.0035 -0.9997>}
 deform_knee_in_base_r {parent=calf_r, loc=<-1.298 5.2957 0.0496>, rot=<0 0 0 -1>}
 deform_knee_in_r {parent=deform_knee_in_base_r, loc=<0 0 0>, rot=<0 0 0 1>}
 deform_knee_out_base_r {parent=calf_r, loc=<0 -4 0>, rot=<0 0 0 -1>}
 deform_knee_out_r {parent=deform_knee_out_base_r, loc=<0 0 0>, rot=<0 0 0 1>}
 thigh_twist_01_r {parent=thigh_r, loc=<21.6303 0 0>, rot=<0.0013 0 0 -1>}
 deform_groin_base_r {parent=thigh_r, loc=<2.1908 -8.1142 1.7442>, rot=<0 0 0 -1>}
 deform_groin_r {parent=deform_groin_base_r, loc=<0 0 0>, rot=<0 0 0 1>}
 deform_glute_base_l {parent=pelvis, loc=<-8.8064 -8.8222 -8.4287>, rot=<0.0275 0.0501 -0.4472 -0.8926>}
 deform_glute_l {parent=deform_glute_base_l, loc=<0 0 0>, rot=<0 0 0 1>}
 deform_glute_base_r {parent=pelvis, loc=<-8.8062 -8.8222 8.4287>, rot=<0.0501 -0.0275 0.8926 -0.4472>}
 deform_glute_r {parent=deform_glute_base_r, loc=<0 0 0>, rot=<0 0 0 1>}
 ground {parent=root, loc=<0 0 0>, rot=<0 0 0 1>}
]
variables [
 L_leg_back {type=double, default=0}
 L_leg_in {type=double, default=0}
 L_leg_out {type=double, default=0}
 L_leg_twist_in {type=double, default=0}
 L_leg_twist_out {type=double, default=0}
 L_leg_up {type=double, default=0}
 R_leg_back {type=double, default=0}
 R_leg_in {type=double, default=0}
 R_leg_out {type=double, default=0}
 R_leg_twist_in {type=double, default=0}
 R_leg_twist_out {type=double, default=0}
 R_leg_up {type=double, default=0}
 cloth_sim {type=bool, default=False}
 crouching {type=bool, default=False}
 skydiving {type=bool, default=False}
 sliding {type=bool, default=False}
 spine_backward {type=double, default=0}
 spine_forward {type=double, default=0}
 spine_left {type=double, default=0}
 spine_right {type=double, default=0}
 upper_spine_backward {type=double, default=0}
 upper_spine_forward {type=double, default=0}
 upper_spine_left {type=double, default=0}
 upper_spine_right {type=double, default=0}
 upper_spine_twist_in {type=double, default=0}
 upper_spine_twist_out {type=double, default=0}
]
