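# Physics asset for the TeddyPermit medium male player body (FE variant), set up for a dynamic coat.
# Capsule bone chains for left, right and back panels hang from small MSTR_coat anchor bodies with limited swing and twist.
# Pelvis, spine, legs and feet carry collision shapes, with no constraints between them.
# Angular-free constraints tie neighboring hem bodies around sides and back, some pairs in both directions, leaving the front open.
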PhysicsAsset /BRCosmetics/Characters/Player/Male/Medium/Bodies/M_MED_TeddyPermit/Meshes/M_MED_TeddyPermit_FE_Physics
bodies {
 MSTR_coat_01_L_01 [{shapeType=Capsule, length=0.505, radius=0.505, shapeName=MSTR_coat_01_L_01_capsule, rotation=<rot 0 -90 0>}]
 MSTR_coat_01_R_01 [{shapeType=Capsule, length=0.505, radius=0.505, shapeName=MSTR_coat_01_R_01_capsule, rotation=<rot 0 -90 0>}]
 MSTR_coat_02_L_01 [{shapeType=Capsule, length=0.505, radius=0.505, shapeName=MSTR_coat_02_L_01_capsule, rotation=<rot 0 -90 0>}]
 MSTR_coat_02_R_01 [{shapeType=Capsule, length=0.505, radius=0.505, shapeName=MSTR_coat_02_R_01_capsule, rotation=<rot 0 -90 180>}]
 MSTR_coat_03_L_01 [{shapeType=Capsule, length=0.505, radius=0.505, shapeName=MSTR_coat_03_L_01_capsule, rotation=<rot 0 -90 0>}]
 MSTR_coat_03_R_01 [{shapeType=Capsule, length=0.505, radius=0.505, shapeName=MSTR_coat_03_R_01_capsule, rotation=<rot 0 -90 0>}]
 MSTR_coat_B_01 [{shapeType=Capsule, length=0.505, radius=0.505, shapeName=MSTR_coat_B_01_capsule, rotation=<rot 0 -90 0>}]
 calf_l [{shapeType=Capsule, length=28, radius=8, center=<-16.1877 -0.3017 0.4251>, rotation=<rot 0 92.4361 -87.3355>}]
 calf_r [{shapeType=Capsule, length=28, radius=8, center=<16.1623 0.3006 0.1211>, rotation=<rot 0 92.4361 92.6645>}]
 dyn_coat_01_L_01 [{
   shapeType=Capsule
   length=12.4516
   radius=3.0143
   shapeName=dyn_coat_01_L_01_capsule
   center=<3.6775 -1.2674 -2.9183>
   rotation=<rot 0 0 -12.2994>
  }]
 dyn_coat_01_L_02 [
  {
   shapeType=Capsule
   length=7.9181
   radius=1.872
   shapeName=dyn_coat_01_L_02_capsule
   center=<5.9397 -1.0313 0.0911>
   rotation=<rot 62.566 -60.3013 -59.8302>
  }
  {
   shapeType=Capsule
   length=7.4983
   radius=1.463
   shapeName=dyn_coat_01_L_02_capsule
   center=<10.7245 -1.0877 0.8655>
   rotation=<rot 67.2976 -127.1771 -124.5536>
  }
  {
   shapeType=Capsule
   length=23.043
   radius=2.7844
   shapeName=dyn_coat_01_L_02_capsule
   center=<14.6344 7.3458 -10.251>
   rotation=<rot 17.9367 -3.7149 -39.5173>
  }
 ]
 dyn_coat_01_L_03 [
  {
   shapeType=Capsule
   length=24.7698
   radius=3.8103
   shapeName=dyn_coat_01_L_03_capsule
   center=<19.5544 5.3772 -10.1884>
   rotation=<rot 0.0001 -0.0001 -34.638>
  }
  {
   shapeType=Capsule
   length=6.5395
   radius=3.0379
   shapeName=dyn_coat_01_L_03_capsule
   center=<6.6432 -0.9072 -1.5806>
   rotation=<rot 43.9971 -24.63 -30.5487>
  }
  {
   shapeType=Capsule
   length=13.2598
   radius=3.2276
   shapeName=dyn_coat_01_L_03_capsule
   center=<14.5874 0.4088 -4.7863>
   rotation=<rot 21.9728 -10.3293 -25.7354>
  }
 ]
 dyn_coat_01_L_04 [
  {
   shapeType=Capsule
   length=10.2154
   radius=4.3187
   shapeName=dyn_coat_01_L_04_capsule
   center=<23.15 2.3983 -2.6011>
   rotation=<rot 0 -0.0001 -25.3642>
  }
  {
   shapeType=Capsule
   length=9.5844
   radius=4.4979
   shapeName=dyn_coat_01_L_04_capsule
   center=<10.8714 1.4576 -2.6591>
   rotation=<rot 0 -0.0001 -35.4059>
  }
 ]
 dyn_coat_01_R_01 [{
   shapeType=Capsule
   length=9.1506
   radius=4.352
   shapeName=dyn_coat_01_R_01_capsule
   center=<2.9651 -2.619 1.2446>
   rotation=<rot 0 0 -169.0874>
  }]
 dyn_coat_01_R_02 [
  {
   shapeType=Capsule
   length=6.8878
   radius=2.0056
   shapeName=dyn_coat_01_R_02_capsule
   center=<9.2112 -1.1982 -0.9521>
   rotation=<rot -65.5954 -80.4588 -96.4921>
  }
  {
   shapeType=Capsule
   length=20.3688
   radius=2.2972
   shapeName=dyn_coat_01_R_02_capsule
   center=<12.4118 1.4466 6.3121>
   rotation=<rot -17.9367 -3.7149 -152.5934>
  }
 ]
 dyn_coat_01_R_03 [
  {
   shapeType=Capsule
   length=25.067
   radius=3.5131
   shapeName=dyn_coat_01_R_03_capsule
   center=<18.5458 1.1867 5.9486>
   rotation=<rot -0.0001 -0.0001 -153.6644>
  }
  {
   shapeType=Capsule
   length=6.11
   radius=3.5401
   shapeName=dyn_coat_01_R_03_capsule
   center=<3.7505 -1.9062 1.8475>
   rotation=<rot -43.9971 -24.63 -149.4512>
  }
  {
   shapeType=Capsule
   length=8.412
   radius=3.4666
   shapeName=dyn_coat_01_R_03_capsule
   center=<12.5867 -1.742 1.4155>
   rotation=<rot -40.0535 -20.6252 -151.1399>
  }
 ]
 dyn_coat_01_R_04 [
  {
   shapeType=Capsule
   length=14.459
   radius=4.3187
   shapeName=dyn_coat_01_R_04_capsule
   center=<22.5646 2.3283 4.3194>
   rotation=<rot 0 -0.0001 -149.0708>
  }
  {
   shapeType=Capsule
   length=13.4857
   radius=4.1966
   shapeName=dyn_coat_01_R_04_capsule
   center=<9.9606 1.9205 4.185>
   rotation=<rot 0 -0.0001 -148.058>
  }
 ]
 dyn_coat_02_L_01 [{shapeType=Capsule, length=12.3323, radius=4, shapeName=dyn_coat_02_L_01_capsule, center=<5 0 0>}]
 dyn_coat_02_L_02 [{
   shapeType=Capsule
   length=11.4619
   radius=6.5973
   shapeName=dyn_coat_02_L_02_capsule
   center=<9.3546 -2.3069 -2.8787>
   rotation=<rot 0 0 -3.1666>
  }]
 dyn_coat_02_L_03 [{
   shapeType=Capsule
   length=35
   radius=9.1219
   shapeName=dyn_coat_02_L_03_capsule
   center=<19.1368 -2.8808 -2.5743>
   rotation=<rot -4.5832 -0.139 1.7388>
  }]
 dyn_coat_02_R_01 [{shapeType=Capsule, length=12.3323, radius=4, shapeName=dyn_coat_02_R_01_capsule, center=<5 -0.3874 0>, rotation=<rot 0 0 180>}]
 dyn_coat_02_R_02 [{
   shapeType=Capsule
   length=14.1814
   radius=5.7053
   shapeName=dyn_coat_02_R_02_capsule
   center=<9.551 -2.1947 3.5652>
   rotation=<rot 0 0 -176.8334>
  }]
 dyn_coat_02_R_03 [{
   shapeType=Capsule
   length=35
   radius=9.1219
   shapeName=dyn_coat_02_R_03_capsule
   center=<19.1368 -2.8808 2.5743>
   rotation=<rot 4.5832 -0.139 175.1523>
  }]
 dyn_coat_03_L_01 [{
   shapeType=Capsule
   length=20
   radius=5
   shapeName=dyn_coat_03_L_01_capsule
   center=<8.7694 -4.3677 0.1335>
   rotation=<rot 0.0001 0.0004 15.1702>
  }]
 dyn_coat_03_L_02 [{shapeType=Capsule, length=20, radius=6.5, shapeName=dyn_coat_03_L_02_capsule, center=<7.5595 -3.4358 1.8164>, rotation=<rot 0 0 35>}]
 dyn_coat_03_L_03 [{shapeType=Capsule, length=35, radius=9, shapeName=dyn_coat_03_L_03_capsule, center=<20 -1.9182 0.3382>, rotation=<rot 0 0 30>}]
 dyn_coat_03_R_01 [{
   shapeType=Capsule
   length=20
   radius=5
   shapeName=dyn_coat_03_R_01_capsule
   center=<8.7694 -4.6963 -0.0848>
   rotation=<rot -0.0011 0.0044 -6.9439>
  }]
 dyn_coat_03_R_02 [{shapeType=Capsule, length=20, radius=6.5, shapeName=dyn_coat_03_R_02_capsule, center=<7.5595 -3.4358 -1.8164>, rotation=<rot 0 0 145>}]
 dyn_coat_03_R_03 [{shapeType=Capsule, length=35, radius=9, shapeName=dyn_coat_03_R_03_capsule, center=<20 -1.9182 -0.3382>, rotation=<rot 0 0 150>}]
 dyn_coat_B_01 [{shapeType=Capsule, length=16.3972, radius=11.6028, shapeName=dyn_coat_B_01_capsule, center=<6.4438 -10.9123 -0.7465>}]
 dyn_coat_B_02 [{shapeType=Capsule, length=21.553, radius=10.447, shapeName=dyn_coat_B_02_capsule, center=<11.4492 -8.6123 0>}]
 dyn_coat_B_03 [{shapeType=Capsule, length=53.2465, radius=11.9019, shapeName=dyn_coat_B_03_capsule, center=<19.4308 -3 0>}]
 foot_l [{
   shapeType=Capsule
   length=15.1657
   radius=7.9942
   shapeName=foot_r_capsule
   center=<-2.0886 5.8594 -0.3533>
   rotation=<rot 0 1.1043 -87.3277>
  }]
 foot_r [{shapeType=Capsule, length=15.1657, radius=7.9942, shapeName=foot_l_capsule, center=<2.0886 -5.8594 0.3533>, rotation=<rot 0 1.1043 92.6723>}]
 pelvis [{shapeType=Capsule, length=7.2955, radius=12.7865, center=<-5.7115 -1.8535 0>}]
 spine_01 [
  {shapeType=Sphere, radius=5.757, shapeName=spine_01_sphere, center=<-0.5698 5.0742 5.71>}
  {shapeType=Capsule, length=6.3221, radius=12.4968, center=<-5.5698 -2.6351 0>}
 ]
 spine_02 [
  {shapeType=Capsule, length=5.8039, radius=12.4968, center=<-5.1339 -3.4242 0>}
  {shapeType=Capsule, length=9.5519, radius=5.9008, shapeName=spine_02_capsule, center=<-0.4853 4.8744 -0.1519>}
 ]
 spine_03 [{shapeType=Capsule, length=3.1042, radius=12.373, shapeName=spine_03_capsule, center=<-0.4533 -1.5074 -0.2034>}]
 spine_04 [{shapeType=Capsule, length=1.7187, radius=14.9998, shapeName=spine_04_capsule, center=<1.6076 -4.454 -0.1873>, rotation=<rot 90 0 0>}]
 thigh_l [
  {shapeType=Capsule, length=37.3259, radius=8.5, center=<-21.9308 1.4932 2.2331>, rotation=<rot 0.0653 -92.4364 -88.7861>}
  {
   shapeType=Capsule
   length=10
   radius=11.25
   shapeName=thigh_l_capsule
   center=<-5.8118 -0.6513 2.4119>
   rotation=<rot -0.1026 -99.3007 -90.6263>
  }
 ]
 thigh_r [
  {shapeType=Capsule, length=37.3259, radius=8.5, center=<21.9308 -1.4932 -2.2331>, rotation=<rot 0.0653 -92.4364 91.2139>}
  {
   shapeType=Capsule
   length=10
   radius=11.25
   shapeName=thigh_r_capsule
   center=<5.8908 1.3782 -2.928>
   rotation=<rot -0.0636 -95.7554 89.3686>
  }
 ]
}
constraints [
 {bone1Name=dyn_coat_01_L_03, bone2Name=dyn_coat_01_L_02, swing1LimitDegrees=45, swing2LimitDegrees=55, twistLimitDegrees=0}
 {bone1Name=dyn_coat_01_R_03, bone2Name=dyn_coat_01_R_02, swing1LimitDegrees=45, swing2LimitDegrees=55, twistLimitDegrees=0}
 {bone1Name=dyn_coat_B_02, bone2Name=dyn_coat_B_01, swing1LimitDegrees=45, swing2LimitDegrees=5, twistLimitDegrees=5}
 {bone1Name=dyn_coat_03_L_02, bone2Name=dyn_coat_03_L_01, swing1LimitDegrees=25, swing2LimitDegrees=25, twistLimitDegrees=10}
 {bone1Name=dyn_coat_02_L_02, bone2Name=dyn_coat_02_L_01, swing1LimitDegrees=25, swing2LimitDegrees=25, twistLimitDegrees=10}
 {bone1Name=dyn_coat_02_R_03, bone2Name=dyn_coat_02_R_02, swing1LimitDegrees=30, swing2LimitDegrees=20, twistLimitDegrees=10}
 {bone1Name=dyn_coat_03_R_02, bone2Name=dyn_coat_03_R_01, swing1LimitDegrees=25, swing2LimitDegrees=25, twistLimitDegrees=10}
 {bone1Name=dyn_coat_02_L_01, bone2Name=MSTR_coat_02_L_01, swing1LimitDegrees=20, swing2LimitDegrees=10, twistLimitDegrees=5}
 {bone1Name=dyn_coat_02_L_03, bone2Name=dyn_coat_02_L_02, swing1LimitDegrees=30, swing2LimitDegrees=20, twistLimitDegrees=10}
 {bone1Name=dyn_coat_02_R_01, bone2Name=MSTR_coat_02_R_01, swing1LimitDegrees=20, swing2LimitDegrees=10, twistLimitDegrees=5}
 {bone1Name=dyn_coat_02_R_02, bone2Name=dyn_coat_02_R_01, swing1LimitDegrees=25, swing2LimitDegrees=25, twistLimitDegrees=10}
 {bone1Name=dyn_coat_03_L_01, bone2Name=MSTR_coat_03_L_01, swing1LimitDegrees=25, swing2LimitDegrees=15, twistLimitDegrees=5}
 {bone1Name=dyn_coat_03_R_03, bone2Name=dyn_coat_03_R_02, swing1LimitDegrees=30, swing2LimitDegrees=20, twistLimitDegrees=10}
 {bone1Name=dyn_coat_03_L_03, bone2Name=dyn_coat_03_L_02, swing1LimitDegrees=30, swing2LimitDegrees=20, twistLimitDegrees=10}
 {bone1Name=dyn_coat_03_R_01, bone2Name=MSTR_coat_03_R_01, swing1LimitDegrees=25, swing2LimitDegrees=15, twistLimitDegrees=5}
 {bone1Name=dyn_coat_B_01, bone2Name=MSTR_coat_B_01, swing1LimitDegrees=50, swing2LimitDegrees=0, twistLimitDegrees=0}
 {bone1Name=dyn_coat_B_03, bone2Name=dyn_coat_B_02, swing1LimitDegrees=35, swing2LimitDegrees=5, twistLimitDegrees=5}
 {bone1Name=dyn_coat_01_L_01, bone2Name=MSTR_coat_01_L_01, swing1LimitDegrees=35, swing2LimitDegrees=40, twistLimitDegrees=10}
 {bone1Name=dyn_coat_01_L_02, bone2Name=dyn_coat_01_L_01, swing1LimitDegrees=40, swing2LimitDegrees=55, twistLimitDegrees=5}
 {bone1Name=dyn_coat_01_L_04, bone2Name=dyn_coat_01_L_03, swing1LimitDegrees=35, swing2LimitDegrees=55, twistLimitDegrees=0}
 {bone1Name=dyn_coat_01_R_02, bone2Name=dyn_coat_01_R_01, swing1LimitDegrees=40, swing2LimitDegrees=55, twistLimitDegrees=5}
 {bone1Name=dyn_coat_01_R_04, bone2Name=dyn_coat_01_R_03, swing1LimitDegrees=35, swing2LimitDegrees=55, twistLimitDegrees=0}
 {bone1Name=dyn_coat_01_R_01, bone2Name=MSTR_coat_01_R_01, swing1LimitDegrees=35, swing2LimitDegrees=40, twistLimitDegrees=10}
 {
  bone1Name=dyn_coat_02_L_03
  bone2Name=dyn_coat_01_L_04
  swing1LimitDegrees=45
  swing1Motion=Free
  swing2LimitDegrees=45
  swing2Motion=Free
  twistLimitDegrees=45
  twistMotion=Free
 }
 {
  bone1Name=dyn_coat_02_L_03
  bone2Name=dyn_coat_03_L_03
  swing1LimitDegrees=45
  swing1Motion=Free
  swing2LimitDegrees=45
  swing2Motion=Free
  twistLimitDegrees=45
  twistMotion=Free
 }
 {
  bone1Name=dyn_coat_B_03
  bone2Name=dyn_coat_03_L_03
  swing1LimitDegrees=45
  swing1Motion=Free
  swing2LimitDegrees=45
  swing2Motion=Free
  twistLimitDegrees=45
  twistMotion=Free
 }
 {
  bone1Name=dyn_coat_01_R_04
  bone2Name=dyn_coat_02_R_03
  swing1LimitDegrees=45
  swing1Motion=Free
  swing2LimitDegrees=45
  swing2Motion=Free
  twistLimitDegrees=45
  twistMotion=Free
 }
 {
  bone1Name=dyn_coat_B_03
  bone2Name=dyn_coat_03_R_03
  swing1LimitDegrees=45
  swing1Motion=Free
  swing2LimitDegrees=45
  swing2Motion=Free
  twistLimitDegrees=45
  twistMotion=Free
 }
 {
  bone1Name=dyn_coat_02_R_03
  bone2Name=dyn_coat_03_R_03
  swing1LimitDegrees=45
  swing1Motion=Free
  swing2LimitDegrees=45
  swing2Motion=Free
  twistLimitDegrees=45
  twistMotion=Free
 }
 {
  bone1Name=dyn_coat_01_L_04
  bone2Name=dyn_coat_02_L_03
  swing1LimitDegrees=45
  swing1Motion=Free
  swing2LimitDegrees=45
  swing2Motion=Free
  twistLimitDegrees=45
  twistMotion=Free
 }
 {
  bone1Name=dyn_coat_03_L_03
  bone2Name=dyn_coat_B_03
  swing1LimitDegrees=45
  swing1Motion=Free
  swing2LimitDegrees=45
  swing2Motion=Free
  twistLimitDegrees=45
  twistMotion=Free
 }
 {
  bone1Name=dyn_coat_03_R_03
  bone2Name=dyn_coat_B_03
  swing1LimitDegrees=45
  swing1Motion=Free
  swing2LimitDegrees=45
  swing2Motion=Free
  twistLimitDegrees=45
  twistMotion=Free
 }
]
