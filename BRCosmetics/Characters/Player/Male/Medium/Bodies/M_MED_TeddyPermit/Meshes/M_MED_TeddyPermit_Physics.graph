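# Physics asset for the TeddyPermit male medium body cosmetic, set up for secondary motion of its coat.
# Seven capsule chains (three per side, one back) hang from tiny MSTR anchor bodies through swing- and twist-limited constraints.
# Free-angular cross-constraints tie neighboring chains at their lower segments, linking the hem as one sheet.
# Capsules on spine, pelvis, legs and feet, plus a huge ground capsule acting as floor, provide collision shapes.
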PhysicsAsset /BRCosmetics/Characters/Player/Male/Medium/Bodies/M_MED_TeddyPermit/Meshes/M_MED_TeddyPermit_Physics
bodies {
 MSTR_coat_01_L_01 [{shapeType=Capsule, length=0.505, radius=0.505, shapeName=MSTR_coat_01_L_01_capsule, rotation=<rot 0 -90 0>}]
 MSTR_coat_01_R_01 [{shapeType=Capsule, length=0.505, radius=0.505, shapeName=MSTR_coat_01_R_01_capsule, rotation=<rot 0 -90 0>}]
 MSTR_coat_02_L_01 [{shapeType=Capsule, length=0.505, radius=0.505, shapeName=MSTR_coat_02_L_01_capsule, rotation=<rot 0 -90 0>}]
 MSTR_coat_02_R_01 [{shapeType=Capsule, length=0.505, radius=0.505, shapeName=MSTR_coat_02_R_01_capsule, rotation=<rot 0 -90 180>}]
 MSTR_coat_03_L_01 [{shapeType=Capsule, length=0.505, radius=0.505, shapeName=MSTR_coat_03_L_01_capsule, rotation=<rot 0 -90 0>}]
 MSTR_coat_03_R_01 [{shapeType=Capsule, length=0.505, radius=0.505, shapeName=MSTR_coat_03_R_01_capsule, rotation=<rot 0 -90 0>}]
 MSTR_coat_B_01 [{shapeType=Capsule, length=0.505, radius=0.505, shapeName=MSTR_coat_B_01_capsule, rotation=<rot 0 -90 0>}]
 calf_l [{shapeType=Capsule, length=27.174, radius=8.826, center=<-16.1877 -0.3017 0.4251>, rotation=<rot 0 92.4361 -87.3355>}]
 calf_r [{shapeType=Capsule, length=28, radius=8, center=<16.1623 0.3006 0.1211>, rotation=<rot 0 92.4361 92.6645>}]
 dyn_coat_01_L_01 [{
   shapeType=Capsule
   length=11.8798
   radius=3.5861
   shapeName=dyn_coat_01_L_01_capsule
   center=<3.6775 -1.1283 -2.888>
   rotation=<rot 0 0 -12.2994>
  }]
 dyn_coat_01_L_02 [
  {
   shapeType=Capsule
   length=7.3719
   radius=1.5215
   shapeName=dyn_coat_01_L_02_capsule
   center=<5.742 -0.7373 -0.9154>
   rotation=<rot 62.566 -60.3013 -73.8578>
  }
  {
   shapeType=Capsule
   length=7.1896
   radius=1.7717
   shapeName=dyn_coat_01_L_02_capsule
   center=<11.0279 -0.6876 -0.3347>
   rotation=<rot 67.2976 -127.1771 -124.5536>
  }
  {
   shapeType=Capsule
   length=24.2793
   radius=5.3587
   shapeName=dyn_coat_01_L_02_capsule
   center=<16.4094 0.9567 -13.3546>
   rotation=<rot 2.5495 7.4098 -15.6663>
  }
 ]
 dyn_coat_01_L_03 [{
   shapeType=Capsule
   length=34.998
   radius=8.3388
   shapeName=dyn_coat_01_L_03_capsule
   center=<10.9804 3.5466 -21.3332>
   rotation=<rot -1.3165 -0.5747 -26.2038>
  }]
 dyn_coat_01_L_04 [
  {
   shapeType=Capsule
   length=25.6651
   radius=4.3187
   shapeName=dyn_coat_01_L_04_capsule
   center=<20.3617 5.0081 -9.3459>
   rotation=<rot -1.9044 1.0987 -29.9912>
  }
  {
   shapeType=Capsule
   length=31.2779
   radius=7.0035
   shapeName=dyn_coat_01_L_04_capsule
   center=<8.1496 8.885 -14.1595>
   rotation=<rot 0 -0.0001 -46.1933>
  }
 ]
 dyn_coat_01_R_01 [{
   shapeType=Capsule
   length=11.8798
   radius=3.5861
   shapeName=dyn_coat_01_R_01_capsule
   center=<3.6775 -1.1283 2.888>
   rotation=<rot 0 0 -167.7006>
  }]
 dyn_coat_01_R_02 [
  {
   shapeType=Capsule
   length=7.3719
   radius=1.5215
   shapeName=dyn_coat_01_R_02_capsule
   center=<5.742 -0.7373 0.9154>
   rotation=<rot -62.566 -60.3013 -106.1422>
  }
  {
   shapeType=Capsule
   length=7.4983
   radius=1.463
   shapeName=dyn_coat_01_R_02_capsule
   center=<10.8384 -1.3691 -0.2557>
   rotation=<rot -67.2976 -127.1771 -55.4464>
  }
  {
   shapeType=Capsule
   length=20.2891
   radius=4.5099
   shapeName=dyn_coat_01_R_02_capsule
   center=<15.0772 1.2089 6.9596>
   rotation=<rot -4.0097 6.8619 -144.1642>
  }
 ]
 dyn_coat_01_R_03 [{
   shapeType=Capsule
   length=33.7879
   radius=8.4866
   shapeName=dyn_coat_01_R_03_capsule
   center=<11.3659 3.1248 15.1103>
   rotation=<rot 1.3165 -0.5747 -142.1308>
  }]
 dyn_coat_01_R_04 [
  {
   shapeType=Capsule
   length=28.4751
   radius=4.3187
   shapeName=dyn_coat_01_R_04_capsule
   center=<22.5646 4.1435 8.0658>
   rotation=<rot 0 -0.0001 -143.4981>
  }
  {
   shapeType=Capsule
   length=30.4209
   radius=5.8
   shapeName=dyn_coat_01_R_04_capsule
   center=<9.9606 5.6236 8.3646>
   rotation=<rot 0 -0.0001 -132.9575>
  }
 ]
 dyn_coat_02_L_01 [{shapeType=Capsule, length=11.7375, radius=4.5948, shapeName=dyn_coat_02_L_01_capsule, center=<5 0 0>}]
 dyn_coat_02_L_02 [{
   shapeType=Capsule
   length=41.6081
   radius=8.6147
   shapeName=dyn_coat_02_L_02_capsule
   center=<10.2901 -4.0678 0.1882>
   rotation=<rot 0 0 2.9528>
  }]
 dyn_coat_02_L_03 [{
   shapeType=Capsule
   length=42.5344
   radius=9.4656
   shapeName=dyn_coat_02_L_03_capsule
   center=<18.4109 -4.5617 -2.2181>
   rotation=<rot -1.2247 -0.0052 2.2746>
  }]
 dyn_coat_02_R_01 [{shapeType=Capsule, length=12.3323, radius=4, shapeName=dyn_coat_02_R_01_capsule, center=<5 0 0>, rotation=<rot 0 0 180>}]
 dyn_coat_02_R_02 [{
   shapeType=Capsule
   length=34.7977
   radius=8.2024
   shapeName=dyn_coat_02_R_02_capsule
   center=<12.2865 -4.0283 0.0892>
   rotation=<rot 0 0 178.0219>
  }]
 dyn_coat_02_R_03 [{
   shapeType=Capsule
   length=40
   radius=11
   shapeName=dyn_coat_02_R_03_capsule
   center=<19.1368 -3.4558 2.5569>
   rotation=<rot 4.5832 -0.139 178.2612>
  }]
 dyn_coat_03_L_01 [{
   shapeType=Capsule
   length=20
   radius=5
   shapeName=dyn_coat_03_L_01_capsule
   center=<8.7694 -4.3677 0.1335>
   rotation=<rot 0.0001 0.0004 15.1702>
  }]
 dyn_coat_03_L_02 [{shapeType=Capsule, length=35, radius=6.5, shapeName=dyn_coat_03_L_02_capsule, center=<7.5595 -3.4358 1.8164>, rotation=<rot 0 0 35>}]
 dyn_coat_03_L_03 [{shapeType=Capsule, length=45, radius=9, shapeName=dyn_coat_03_L_03_capsule, center=<20 -1.9182 0.3382>, rotation=<rot 0 0 30>}]
 dyn_coat_03_R_01 [{
   shapeType=Capsule
   length=20
   radius=5
   shapeName=dyn_coat_03_R_01_capsule
   center=<8.7694 -4.6963 -0.0848>
   rotation=<rot -0.0011 0.0044 -6.9439>
  }]
 dyn_coat_03_R_02 [{shapeType=Capsule, length=35, radius=6.5, shapeName=dyn_coat_03_R_02_capsule, center=<7.5595 -3.4358 -1.8164>, rotation=<rot 0 0 145>}]
 dyn_coat_03_R_03 [{shapeType=Capsule, length=45, radius=9, shapeName=dyn_coat_03_R_03_capsule, center=<20 -1.9182 -0.3382>, rotation=<rot 0 0 150>}]
 dyn_coat_B_01 [{shapeType=Capsule, length=16.3972, radius=11.6028, shapeName=dyn_coat_B_01_capsule, center=<6.4438 -10.9123 -0.7465>}]
 dyn_coat_B_02 [{shapeType=Capsule, length=21.553, radius=10.447, shapeName=dyn_coat_B_02_capsule, center=<11.4492 -8.6123 0>}]
 dyn_coat_B_03 [{shapeType=Capsule, length=53.2465, radius=11.9019, shapeName=dyn_coat_B_03_capsule, center=<19.4308 -3 0>}]
 foot_l [{
   shapeType=Capsule
   length=15.1657
   radius=7.9942
   shapeName=foot_r_capsule
   center=<-2.0886 5.8594 -0.3533>
   rotation=<rot 0 1.1043 -87.3277>
  }]
 foot_r [{shapeType=Capsule, length=15.1657, radius=7.9942, shapeName=foot_l_capsule, center=<2.0886 -5.8594 0.3533>, rotation=<rot 0 1.1043 92.6723>}]
 ground [{shapeType=Capsule, length=0.5056, radius=337.6725, shapeName=floor_joint_capsule, center=<2.6954 -15.6715 -335>}]
 pelvis [{shapeType=Capsule, length=8.1581, radius=12.7865, center=<-5.7115 -1.0731 0>}]
 spine_01 [{shapeType=Capsule, length=9.3871, radius=12.4968, center=<-5.5698 -2.6351 0>}]
 spine_02 [{shapeType=Capsule, length=4.3714, radius=13.4462, center=<-8.6453 -2.6351 1.0579>}]
 spine_03 [{shapeType=Capsule, length=6.4677, radius=12.4529, shapeName=spine_03_capsule, center=<-5.194 -1.5784 0>}]
 spine_04 [{shapeType=Capsule, length=4.5663, radius=13.9983, shapeName=spine_04_capsule, center=<0 -3.7983 0>, rotation=<rot 90 0 0>}]
 thigh_l [
  {shapeType=Capsule, length=37.3259, radius=8.5, center=<-21.8814 3.0753 1.4424>, rotation=<rot 0.0653 -92.4364 -86.5939>}
  {
   shapeType=Capsule
   length=10.2557
   radius=10.9943
   shapeName=thigh_l_capsule
   center=<-5.7442 -0.3487 1.6698>
   rotation=<rot -0.0733 -96.6294 -92.7386>
  }
 ]
 thigh_r [
  {shapeType=Capsule, length=37.3259, radius=8.5, center=<24.9577 -0.3857 -1.1421>, rotation=<rot -0.0106 -88.8622 88.7844>}
  {shapeType=Capsule, length=10, radius=11.25, shapeName=thigh_r_capsule, center=<5.889 1.1328 -1.2594>, rotation=<rot 0 -90 -90.6346>}
 ]
}
constraints [
 {bone1Name=dyn_coat_01_L_03, bone2Name=dyn_coat_01_L_02, swing1LimitDegrees=55, swing2LimitDegrees=55, twistLimitDegrees=30}
 {bone1Name=dyn_coat_01_R_03, bone2Name=dyn_coat_01_R_02, swing1LimitDegrees=50, swing2LimitDegrees=50, twistLimitDegrees=30}
 {bone1Name=dyn_coat_B_02, bone2Name=dyn_coat_B_01, swing1LimitDegrees=45, swing2LimitDegrees=5, twistLimitDegrees=5}
 {bone1Name=dyn_coat_03_L_02, bone2Name=dyn_coat_03_L_01, swing1LimitDegrees=25, swing2LimitDegrees=25, twistLimitDegrees=10}
 {bone1Name=dyn_coat_02_L_02, bone2Name=dyn_coat_02_L_01, swing1LimitDegrees=25, swing2LimitDegrees=25, twistLimitDegrees=10}
 {bone1Name=dyn_coat_02_R_03, bone2Name=dyn_coat_02_R_02, swing1LimitDegrees=30, swing2LimitDegrees=20, twistLimitDegrees=10}
 {bone1Name=dyn_coat_03_R_02, bone2Name=dyn_coat_03_R_01, swing1LimitDegrees=25, swing2LimitDegrees=25, twistLimitDegrees=10}
 {bone1Name=dyn_coat_02_L_01, bone2Name=MSTR_coat_02_L_01, swing1LimitDegrees=20, swing2LimitDegrees=10, twistLimitDegrees=5}
 {bone1Name=dyn_coat_02_L_03, bone2Name=dyn_coat_02_L_02, swing1LimitDegrees=30, swing2LimitDegrees=20, twistLimitDegrees=10}
 {bone1Name=dyn_coat_02_R_01, bone2Name=MSTR_coat_02_R_01, swing1LimitDegrees=20, swing2LimitDegrees=10, twistLimitDegrees=5}
 {bone1Name=dyn_coat_02_R_02, bone2Name=dyn_coat_02_R_01, swing1LimitDegrees=25, swing2LimitDegrees=25, twistLimitDegrees=10}
 {bone1Name=dyn_coat_03_L_01, bone2Name=MSTR_coat_03_L_01, swing1LimitDegrees=25, swing2LimitDegrees=15, twistLimitDegrees=5}
 {bone1Name=dyn_coat_03_R_03, bone2Name=dyn_coat_03_R_02, swing1LimitDegrees=30, swing2LimitDegrees=20, twistLimitDegrees=10}
 {bone1Name=dyn_coat_03_L_03, bone2Name=dyn_coat_03_L_02, swing1LimitDegrees=30, swing2LimitDegrees=20, twistLimitDegrees=10}
 {bone1Name=dyn_coat_03_R_01, bone2Name=MSTR_coat_03_R_01, swing1LimitDegrees=25, swing2LimitDegrees=15, twistLimitDegrees=5}
 {bone1Name=dyn_coat_B_01, bone2Name=MSTR_coat_B_01, swing1LimitDegrees=50, swing2LimitDegrees=0, twistLimitDegrees=0}
 {bone1Name=dyn_coat_B_03, bone2Name=dyn_coat_B_02, swing1LimitDegrees=35, swing2LimitDegrees=5, twistLimitDegrees=5}
 {bone1Name=dyn_coat_01_L_01, bone2Name=MSTR_coat_01_L_01, swing1LimitDegrees=15, swing2LimitDegrees=20, twistLimitDegrees=5}
 {bone1Name=dyn_coat_01_L_02, bone2Name=dyn_coat_01_L_01, swing1LimitDegrees=50, swing2LimitDegrees=50, twistLimitDegrees=20}
 {bone1Name=dyn_coat_01_L_04, bone2Name=dyn_coat_01_L_03, swing1LimitDegrees=50, swing2LimitDegrees=50, twistLimitDegrees=30}
 {bone1Name=dyn_coat_01_R_02, bone2Name=dyn_coat_01_R_01, swing1LimitDegrees=50, swing2LimitDegrees=50, twistLimitDegrees=20}
 {bone1Name=dyn_coat_01_R_04, bone2Name=dyn_coat_01_R_03, swing1LimitDegrees=50, swing2LimitDegrees=50, twistLimitDegrees=30}
 {bone1Name=dyn_coat_01_R_01, bone2Name=MSTR_coat_01_R_01, swing1LimitDegrees=15, swing2LimitDegrees=20, twistLimitDegrees=5}
 {
  bone1Name=dyn_coat_01_L_04
  bone2Name=dyn_coat_02_L_03
  swing1LimitDegrees=45
  swing1Motion=Free
  swing2LimitDegrees=45
  swing2Motion=Free
  twistLimitDegrees=45
  twistMotion=Free
 }
 {
  bone1Name=dyn_coat_01_R_04
  bone2Name=dyn_coat_02_R_03
  swing1LimitDegrees=45
  swing1Motion=Free
  swing2LimitDegrees=45
  swing2Motion=Free
  twistLimitDegrees=45
  twistMotion=Free
 }
 {
  bone1Name=dyn_coat_02_L_03
  bone2Name=dyn_coat_03_L_03
  swing1LimitDegrees=45
  swing1Motion=Free
  swing2LimitDegrees=45
  swing2Motion=Free
  twistLimitDegrees=45
  twistMotion=Free
 }
 {
  bone1Name=dyn_coat_02_R_03
  bone2Name=dyn_coat_03_R_03
  swing1LimitDegrees=45
  swing1Motion=Free
  swing2LimitDegrees=45
  swing2Motion=Free
  twistLimitDegrees=45
  twistMotion=Free
 }
 {
  bone1Name=dyn_coat_03_R_03
  bone2Name=dyn_coat_B_03
  swing1LimitDegrees=45
  swing1Motion=Free
  swing2LimitDegrees=45
  swing2Motion=Free
  twistLimitDegrees=45
  twistMotion=Free
 }
 {
  bone1Name=dyn_coat_03_L_03
  bone2Name=dyn_coat_B_03
  swing1LimitDegrees=45
  swing1Motion=Free
  swing2LimitDegrees=45
  swing2Motion=Free
  twistLimitDegrees=45
  twistMotion=Free
 }
 {
  bone1Name=dyn_coat_01_R_03
  bone2Name=dyn_coat_02_R_02
  swing1LimitDegrees=45
  swing1Motion=Free
  swing2LimitDegrees=45
  swing2Motion=Free
  twistLimitDegrees=45
  twistMotion=Free
 }
 {
  bone1Name=dyn_coat_01_L_03
  bone2Name=dyn_coat_02_L_02
  swing1LimitDegrees=45
  swing1Motion=Free
  swing2LimitDegrees=45
  swing2Motion=Free
  twistLimitDegrees=45
  twistMotion=Free
 }
]
