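# Animation Blueprint for the M_MED_TeddyPermit character body, adding secondary motion to its clothing.
# The copied parent pose runs through a Control Rig, sleeve and ribbon AnimDynamics and bone offsets into a PreSim cache.
# With cloth enabled that pose outputs directly; otherwise RigidBody simulation uses front-end or in-game physics assets.
# The event graph drives cloth or gravity-override parameters, interpolates slope warp values and lowers RigidBody alpha while skydiving.
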AnimBlueprint /BRCosmetics/Characters/Player/Male/Medium/Bodies/M_MED_TeddyPermit/Meshes/M_MED_TeddyPermit_AnimBP
target_skeleton=&/M_MED_TeddyPermit_Skeleton
graphs [
 AnimGraph {nodes [
   n1 = RigidBody!(
    ComponentPose=Local To Component(LocalPose=Use cached pose 'PreSim'())
    Alpha=Get RBAN_alpha() -> (RBAN_alpha:double)
    OverrideWorldGravity:Vector="0,0,0"
    ExternalForce
    SimSpaceSettings:Sim Space Settings Structure=(WorldAlpha=1,VelocityScaleZ=0.4,DampingAlpha=1,MaxLinearVelocity=300,MaxAngularVelocity=10,MaxLinearAcceleration=1000,MaxAngularAcceleration=40,ExternalLinearDragV=(X=0.3,Y=0.3,Z=0.3),ExternalLinearVelocity=(X=0,Y=0,Z=0),ExternalAngularVelocity=(X=0,Y=0,Z=0))
   ) -> (Pose) {
    bClampLinearTranslationLimitToRefPose=true
    bOverrideWorldGravity=true
    cachedBoundsScale=1
    overridePhysicsAsset=&/M_MED_TeddyPermit_Physics
    simulationSpace=BaseBoneSpace
    alphaBoolBlend {blendOption=Linear}
    baseBoneRef {boneName=root}
    componentAppliedLinearAccClamp=<0 0 0>
    overrideWorldGravity {z=-980}
    simSpaceSettings {
     maxAngularAcceleration=40
     maxAngularVelocity=10
     maxLinearAcceleration=1000
     maxLinearVelocity=300
     velocityScaleZ=0.4
     worldAlpha=1
     externalLinearDragV=<0.3 0.3 0.3>
    }
   }
   n2 = Component To Local(ComponentPose=@n1)
   n3 = RigidBody(
    ComponentPose=Local To Component(LocalPose=Use cached pose 'PreSim'())
    Alpha=Get RBAN_alpha()
    OverrideWorldGravity="0,0,0"
    SimSpaceSettings=(WorldAlpha=1,VelocityScaleZ=0.4,DampingAlpha=1,MaxLinearVelocity=300,MaxAngularVelocity=10,MaxLinearAcceleration=1000,MaxAngularAcceleration=40,ExternalLinearDragV=(X=0.3,Y=0.3,Z=0.3),ExternalLinearVelocity=(X=0,Y=0,Z=0),ExternalAngularVelocity=(X=0,Y=0,Z=0))
   ) {
    bClampLinearTranslationLimitToRefPose=true
    bOverrideWorldGravity=true
    cachedBoundsScale=1
    overridePhysicsAsset=&/M_MED_TeddyPermit_FE_Physics
    simulationSpace=BaseBoneSpace
    alphaBoolBlend {blendOption=Linear}
    baseBoneRef {boneName=root}
    componentAppliedLinearAccClamp=<0 0 0>
    overrideWorldGravity {z=-980}
    simSpaceSettings {
     maxAngularAcceleration=40
     maxAngularVelocity=10
     maxLinearAcceleration=1000
     maxLinearVelocity=300
     velocityScaleZ=0.4
     worldAlpha=1
     externalLinearDragV=<0.3 0.3 0.3>
    }
   }
   n4 = Component To Local(ComponentPose=@n3)
   n5 = Get bIsInFrontEnd()
   Output Pose(Result=Blend Poses by bool(
     BlendPose_0=Use cached pose 'PreSim'()
     BlendPose_1=Blend Poses by bool(BlendPose_0=@n4, BlendPose_1=@n2, BlendTime_0=0.2, BlendTime_1=0.5, bActiveValue=@n5) {blendType=Linear}
     BlendTime_0=0.2
     BlendTime_1=0.5
     bActiveValue=Get bClothEnabled()
    ) {blendType=Linear}) {layerGroup=None}
   n6 = Control Rig!(
    Source=Copy Pose From Mesh() {bUseAttachedParent=true}
    Alpha
    crouching:bool=Get bIsCrouching()
    cloth_sim:bool=Get bClothEnabled()
    sliding:bool=Get sliding() -> (sliding:bool)
    skydiving:bool=Get bIsSkyDiving()
   ) -> (Pose) {controlRigAssetReference {blueprintRigClass=&/M_MED_TeddyPermit_CtrlRig.M_MED_TeddyPermit_CtrlRig_C}}
   n7 = Anim Dynamics - Bone: dyn_sleeve_l!(ComponentPose=Local To Component(LocalPose=@n6), Alpha=0.2, GravityOverride:Vector="0,0,-980", ExternalForce="0,0,0") -> (Pose) {
    angularDampingOverride=0.999
    angularSpringConstant=4500
    bAngularSpring=true
    bUseGravityOverride=true
    gravityScale=0.35
    simulationSpace=BoneRelative
    boundBone {boneName=dyn_sleeve_l}
    componentAppliedLinearAccClamp=<1000 1000 1000>
    componentLinearAccScale=<0.1 0.1 0.1>
    componentLinearVelScale=<0.1 0.1 0.1>
    gravityOverride {z=-980}
    physicsBodyDefinitions [{
      boundBone {boneName=dyn_sleeve_l}
      boxExtents=<40 30 30>
      constraintSetup {angularLimitsMax=<10 10 15>, angularLimitsMin=<-10 -10 -15>, angularTarget=<1 0 0>}
      localJointOffset=<1 0 0>
     }]
    relativeSpaceBone {boneName=lowerarm_l}
    simSpaceSettings {maxAngularAcceleration=35, maxAngularVelocity=7, simSpaceAngularAlpha=1}
   }
   n8 = Anim Dynamics - Bone: dyn_sleeve_r!(ComponentPose=@n7, Alpha=0.2, GravityOverride:Vector="0,0,-980", ExternalForce="0,0,0") -> (Pose) {
    angularDampingOverride=0.999
    angularSpringConstant=5000
    bAngularSpring=true
    bUseGravityOverride=true
    gravityScale=0.35
    simulationSpace=BoneRelative
    boundBone {boneName=dyn_sleeve_r}
    componentAppliedLinearAccClamp=<1000 1000 1000>
    componentLinearAccScale=<0.1 0.1 0.1>
    componentLinearVelScale=<0.2 0.2 0.2>
    gravityOverride {z=-980}
    physicsBodyDefinitions [{
      boundBone {boneName=dyn_sleeve_r}
      boxExtents=<40 30 30>
      constraintSetup {angularLimitsMax=<10 20 10>, angularLimitsMin=<-10 -10 -10>, angularTarget=<1 0 0>}
      localJointOffset=<1 0 0>
     }]
    relativeSpaceBone {boneName=lowerarm_r}
    simSpaceSettings {maxAngularAcceleration=35, maxAngularVelocity=7, simSpaceAngularAlpha=1}
   }
   n9 = Anim Dynamics - Bone: dyn_ribbon_A_l!(ComponentPose=@n8, Alpha=1, GravityOverride:Vector="0,0,-980", ExternalForce="0,0,0") -> (Pose) {
    angularDampingOverride=0.999
    angularSpringConstant=5000
    bAngularSpring=true
    bLinearSpring=true
    bUseGravityOverride=true
    gravityScale=0.35
    linearSpringConstant=3500
    simulationSpace=BoneRelative
    boundBone {boneName=dyn_ribbon_A_l}
    componentAppliedLinearAccClamp=<1000 1000 1000>
    componentLinearAccScale=<0.1 0.1 0.1>
    componentLinearVelScale=<0.3 0.3 0.3>
    gravityOverride {z=-980}
    physicsBodyDefinitions [{
      boundBone {boneName=dyn_ribbon_A_l}
      boxExtents=<40 30 30>
      constraintSetup {angularLimitsMax=<0 10 45>, angularLimitsMin=<0 -10 0>, angularTarget=<1 0 0>, linearAxesMin=<-0.5 0 0>}
      localJointOffset=<1 0 0>
     }]
    relativeSpaceBone {boneName=pelvis}
    simSpaceSettings {maxAngularAcceleration=35, maxAngularVelocity=7, simSpaceAngularAlpha=1}
   }
   n10 = Anim Dynamics - Bone: dyn_ribbon_A_r!(ComponentPose=@n9, Alpha=0.8, GravityOverride:Vector="0,0,-980", ExternalForce="0,0,0") -> (Pose) {
    angularDampingOverride=0.999
    angularSpringConstant=5000
    bAngularSpring=true
    bLinearSpring=true
    bUseGravityOverride=true
    gravityScale=0.35
    linearSpringConstant=3500
    simulationSpace=BoneRelative
    boundBone {boneName=dyn_ribbon_A_r}
    componentAppliedLinearAccClamp=<1000 1000 1000>
    componentLinearAccScale=<0.1 0.1 0.1>
    componentLinearVelScale=<0.3 0.3 0.3>
    gravityOverride {z=-980}
    physicsBodyDefinitions [{
      boundBone {boneName=dyn_ribbon_A_r}
      boxExtents=<40 30 30>
      constraintSetup {angularLimitsMax=<0 10 45>, angularLimitsMin=<0 -10 0>, angularTarget=<1 0 0>, linearAxesMin=<-0.5 0 0>}
      localJointOffset=<1 0 0>
     }]
    relativeSpaceBone {boneName=pelvis}
    simSpaceSettings {maxAngularAcceleration=35, maxAngularVelocity=7, simSpaceAngularAlpha=1}
   }
   n11 = Anim Dynamics - Bone: dyn_ribbon_B_l!(ComponentPose=@n10, Alpha=0.75, GravityOverride:Vector="0,0,-980", ExternalForce="0,0,0") -> (Pose) {
    alpha=0.75
    angularDampingOverride=0.999
    angularSpringConstant=3000
    bAngularSpring=true
    bLinearSpring=true
    bUseGravityOverride=true
    gravityScale=0.35
    linearSpringConstant=3500
    simulationSpace=BoneRelative
    boundBone {boneName=dyn_ribbon_B_l}
    componentAppliedLinearAccClamp=<1000 1000 1000>
    componentLinearAccScale=<0.1 0.1 0.1>
    componentLinearVelScale=<0.3 0.3 0.3>
    gravityOverride {z=-980}
    physicsBodyDefinitions [{
      boundBone {boneName=dyn_ribbon_B_l}
      boxExtents=<40 30 30>
      constraintSetup {angularLimitsMax=<0 10 45>, angularLimitsMin=<0 -10 0>, angularTarget=<1 0 0>, linearAxesMin=<-0.5 0 0>}
      localJointOffset=<1 0 0>
     }]
    relativeSpaceBone {boneName=pelvis}
    simSpaceSettings {maxAngularAcceleration=35, maxAngularVelocity=7, simSpaceAngularAlpha=1}
   }
   n12 = Anim Dynamics - Bone: dyn_ribbon_B_r!(ComponentPose=@n11, Alpha=0.5, GravityOverride:Vector="0,0,-980", ExternalForce="0,0,0") -> (Pose) {
    alpha=0.75
    angularDampingOverride=0.999
    angularSpringConstant=3000
    bAngularSpring=true
    bLinearSpring=true
    bUseGravityOverride=true
    gravityScale=0.35
    linearSpringConstant=3500
    simulationSpace=BoneRelative
    boundBone {boneName=dyn_ribbon_B_r}
    componentAppliedLinearAccClamp=<1000 1000 1000>
    componentLinearAccScale=<0.1 0.1 0.1>
    componentLinearVelScale=<0.3 0.3 0.3>
    gravityOverride {z=-980}
    physicsBodyDefinitions [{
      boundBone {boneName=dyn_ribbon_B_r}
      boxExtents=<40 30 30>
      constraintSetup {angularLimitsMax=<0 10 45>, angularLimitsMin=<0 -10 0>, angularTarget=<1 0 0>, linearAxesMin=<-0.5 0 0>}
      localJointOffset=<1 0 0>
     }]
    relativeSpaceBone {boneName=pelvis}
    simSpaceSettings {maxAngularAcceleration=35, maxAngularVelocity=7, simSpaceAngularAlpha=1}
   }
   n13 = "Transform (Modify) Bone - Bone: backpack_collision"(ComponentPose=@n12, Alpha=1, Translation="0,0,200", Rotation="0,0,0", Scale="1,1,1") {
    rotationMode=BMM_Additive
    rotationSpace=BCS_BoneSpace
    translationMode=BMM_Additive
    translationSpace=BCS_BoneSpace
    boneToModify {boneName=backpack_collision}
    translation {z=200}
   }
   n14 = "Transform (Modify) Bone - Bone: ground"!(ComponentPose=@n13, bAlphaBoolEnabled:bool=True, Translation="0,0,0", Rotation="0,0,0", Scale="1,1,1") -> (Pose) {
    alphaInputType=Bool
    rotationMode=BMM_Replace
    rotationSpace=BCS_WorldSpace
    translationMode=BMM_Replace
    translationSpace=BCS_BoneSpace
    alphaBoolBlend {blendOption=Linear}
    boneToModify {boneName=ground}
   }
   Save cached pose 'PreSim'(Pose=Component To Local(ComponentPose=@n14)) {cacheName=PreSim}
  ]}
 EventGraph {nodes [
   n1 = Event BlueprintUpdateAnimation()
   > Set DeltaTimeLocal(DeltaTimeLocal:double=@n1.DeltaTimeX) -> (Output_Get:double)
   > Set sliding(sliding:bool=OR Boolean(A=Get bIsKneeSliding(), B=Get bIsBaseballSliding())) -> (Output_Get:bool)
   > n2 = Branch(Condition=Get bIsSkyDiving())
   > n3 = Set RBAN_alpha(RBAN_alpha:double=0.7) -> (Output_Get:double)
   n4 = Branch(execute=[@n9.then, @n3.then], Condition=Get bClothEnabled())
   > n5 = Sequence()
   Handle Slope Warp!(execute=[@n5.then_0, @n7.then_1], self, DeltaTime:double=Get DeltaTimeLocal() -> (DeltaTimeLocal:double)) -> (then)
   n6 = "Cloth Controller Function (Parameter Asset) (Nativized)"!(
    execute=@n5.then_1
    CharacterPartAnimInstance:Custom Character Part Anim Instance Object Reference=Self-Reference()
    ClothingAssetName:Name=M_MED_TeddyPermit_Cloth_Clothing
    ClothCoreSettings:"Cloth Core Settings Structure (by ref)"=Get Cloth Core Settings Native() -> (Cloth Core Settings Native:Cloth Core Settings Structure)
    WindGustLocalSettings:"Wind Gust Settings Structure (by ref)"=Get Wind Gust Local Settings Native() -> (Wind Gust Local Settings Native:Wind Gust Settings Structure)
    ClothUpAxis:Vector="1,0,0"
    LinVelScaleFwdBwdClamp:float=0
    LinVelScaleUpDwnClamp:float=0.7
    DeltaTime:float=Get DeltaTimeLocal()
    GameTimeInSeconds:float=GetGameTimeInSeconds()
    StateLogicLayerName:Name=none
    Parameters:Fort Character Dynamics Parameters Object Reference=&/M_MED_TeddyPermit_Parameters
   ) -> (
    then
    OutCurrentState:Name=None
    OutClothParameters:Cloth Parameters Structure
    OutClothCoreSettings:Cloth Core Settings Structure
    OutWindGustSettings:Wind Gust Parameters Structure
    OutWindGustLocalSettings:Wind Gust Settings Structure
    OutWindVelocity:Vector
    OutLinearVelVector:Vector
   )
   > Set Wind Gust Local Settings Native Out(Wind Gust Local Settings Native Out:Wind Gust Settings Structure=@n6.OutWindGustLocalSettings) -> (Output_Get:Wind Gust Settings Structure)
   n7 = Sequence(execute=@n4.else)
   > n8 = "Gravity Override State Function (Parameter Asset) (Nativized)"(
    CharacterPartAnimInstance=Self-Reference()
    GameTimeInSeconds=GetGameTimeInSeconds()
    Parameters=&M_MED_TeddyPermit_Parameters
   )
   > Set Gravity Override(Gravity Override=@n8)
   > Set RBAN Settings Out(RBAN Settings Out=@n8.OutRigidBodyAnimNodeParameters)
   n9 = Set RBAN_alpha(execute=@n2.else, RBAN_alpha=1)
   Event OnCharacterCustomizationLoaded()
   n10 = CheckForBackbling()
   > Set Is Not Wearing Backbling(Is Not Wearing Backbling:bool=NOT Boolean(A=OR Boolean(A=@n10, B=@n10.bIsBackblingLong))) -> (Output_Get:bool)
  ]}
 Handle Slope Warp {nodes [
   n1 = Handle Slope Warp!() -> (then, DeltaTime:double=~) {is=title K2Node_FunctionEntry}
   > n2 = Sequence()
   n3 = Get bIsOnGround() -> (bIsOnGround:bool)
   > n4 = Branch(Condition=OR Boolean!(A=NOT Boolean(A=@n3), B=Get bIsPlayingEmote(), C:bool=Get bIsInVehicle() -> (bIsInVehicle:bool)) -> (ReturnValue))
   > Set Target Info Input Root Slope Translation(Target Info Input Root Slope Translation:Vector="0,0,-100") -> (Output_Get:Vector)
   Set Slope Info Input Slope Rotation(
    execute=@n4.else
    Slope Info Input Slope Rotation:Rotator=RInterpTo!(
     Current:Rotator=Get Slope Info Input Slope Rotation() -> (Slope Info Input Slope Rotation:Rotator)
     Target:Rotator=Get!() -> (
      SlopeInfoInput_RootSlopeTranslation:Vector=~
      SlopeInfoInput_LeftFootSlopeTranslation:Vector=~
      SlopeInfoInput_RightFootSlopeTranslation:Vector=~
      SlopeInfoInput_SlopeRotation:Rotator=~
     ).SlopeInfoInput_SlopeRotation
     DeltaTime:float=@n1
     InterpSpeed:float=7
    ) -> (ReturnValue:Rotator)
   ) -> (Output_Get:Rotator)
   > Set Target Info Input Root Slope Translation(Target Info Input Root Slope Translation="0, 0, 0")
   Reroute Node(InputPin=@n2.then_1)
   Reroute Node()
   > Set Slope Info Input Root Slope Translation(Slope Info Input Root Slope Translation:Vector=VInterpTo!(
     Current:Vector=Get Slope Info Input Root Slope Translation() -> (Slope Info Input Root Slope Translation:Vector)
     Target:Vector=Get Target Info Input Root Slope Translation() -> (Target Info Input Root Slope Translation:Vector)
     DeltaTime:float=@n1
     InterpSpeed:float=10
    ) -> (ReturnValue:Vector)) -> (Output_Get:Vector)
  ]}
]
variables [
 Cloth Core Settings Native {
  type=ClothCoreSettings
  default {
   collisionThickness=0
   dampingCoefficient=0.03
   frictionCoefficient=0.4
   localDampingCoefficient=0
   numberIterations=2
   numberSubsteps=1
   selfCollisionThickness=0
   animDriveDamping=<0 1>
   animDriveStiffness=<0 1>
   areaStiffness=<0.95 1>
   bendingStiffness=<0.75 1>
   edgeStiffness=<0.85 1>
   tetherStiffness=<1 1>
  }
 }
 DeltaTimeLocal {type=number}
 Gravity Override {type=Vector}
 Is Not Wearing Backbling {type=boolean}
 RBAN Settings Out {
  type=FortRigidBodyAnimNodeParameters
  default {simSpaceSettings {dampingAlpha=0, maxAngularAcceleration=16, maxAngularVelocity=4, maxLinearAcceleration=800, maxLinearVelocity=200, velocityScaleZ=0.5}}
 }
 RBAN_alpha {type=number}
 sliding {type=boolean}
 Slope Info Input Root Slope Translation {type=Vector}
 Slope Info Input Slope Rotation {type=Rotator}
 Target Info Input Root Slope Translation {type=Vector}
 Wind Gust Local Settings Native {type=WindGustSettings, default {elapsedTime=0, finalGustStrengthLocal=0, gustDuration=0, gustStrength=0, timeBetweenGusts=0}}
 Wind Gust Local Settings Native Out {type=WindGustSettings, default {elapsedTime=0, finalGustStrengthLocal=0, gustDuration=0, gustStrength=0, timeBetweenGusts=0}}
]
