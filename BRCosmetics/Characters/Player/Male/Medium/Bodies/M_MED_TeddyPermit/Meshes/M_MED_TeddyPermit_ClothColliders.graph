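# Physics asset of cloth collider shapes for the TeddyPermit medium male player body, named for cloth simulation to collide against.
# Capsules cover the pelvis, spine, neck, clavicles, thighs and calves, with a sphere for the head; arms have no bodies and feet have no shapes.
# A backpack capsule adds bulk, and a huge capsule on the ground bone approximates the floor.
# It defines no constraints, so it is not a jointed ragdoll.
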PhysicsAsset /BRCosmetics/Characters/Player/Male/Medium/Bodies/M_MED_TeddyPermit/Meshes/M_MED_TeddyPermit_ClothColliders
bodies {
 backpack_collision [{
   shapeType=Capsule
   length=12.4856
   radius=16.4572
   shapeName=backpack_collision_capsule
   center=<0 -22.7563 0.2656>
   rotation=<rot 8.0973 -90 0>
  }]
 calf_l [{shapeType=Capsule, length=26.946, radius=7.054, shapeName=calf_l_capsule, center=<-18.4519 -0.4891 0>, rotation=<rot 90 0 0>}]
 calf_r [{shapeType=Capsule, length=26.9739, radius=7.0261, shapeName=calf_r_capsule, center=<18.4519 0.2612 0>, rotation=<rot 90 0 0>}]
 clavicle_l [{shapeType=Capsule, length=8, radius=7, center=<7.9335 -1.6216 0.0756>, rotation=<rot 79.3028 -1.4097 174.1847>}]
 clavicle_r [{shapeType=Capsule, length=8, radius=7, center=<-7.9788 2.6227 -0.3152>, rotation=<rot 79.3028 -1.4093 -5.8148>}]
 foot_l []
 foot_r []
 ground [{
   shapeType=Capsule
   length=0.5056
   radius=337.6725
   shapeName=floor_joint_capsule
   center=<2.6954 -12.4561 -333.6685>
   rotation=<rot 0 90 -90>
  }]
 head [{shapeType=Sphere, radius=17.3125, center=<11.028 -0.174 0>}]
 neck_02 [{shapeType=Capsule, length=13.4375, radius=6, center=<-0.2642 -0.3408 0>, rotation=<rot 90 -171.1193 -175.3978>}]
 pelvis [{shapeType=Capsule, length=6.7958, radius=12.6889, center=<-3 -2.9673 0>}]
 spine_01 [
  {shapeType=Capsule, length=5.8267, radius=10.8078, center=<-1.7965 -0.9146 0>}
  {shapeType=Capsule, length=13.0518, radius=8.5675, shapeName=spine_01_capsule, center=<0 -4.9762 0>}
 ]
 spine_02 [
  {shapeType=Capsule, length=19.2214, radius=10, center=<2.5024 -0.3249 3.9825>, rotation=<rot 90 0 0>}
  {shapeType=Capsule, length=19.7007, radius=10, center=<2.4704 -0.325 -3.9825>, rotation=<rot 90 0 0>}
  {shapeType=Capsule, length=15.1889, radius=9.0712, shapeName=spine_02_capsule, center=<4.0169 2.9008 -0.0799>, rotation=<rot 90 0.2113 0>}
 ]
 spine_03 [{shapeType=Capsule, length=8.7971, radius=9.4136, center=<0.1445 -0.8042 0.0001>}]
 spine_04 [{shapeType=Capsule, length=11.2738, radius=10.3104, center=<3.9819 -0.0168 0>}]
 spine_05 [
  {shapeType=Capsule, length=11.948, radius=10, center=<-3.3159 -0.2823 0>}
  {shapeType=Capsule, length=13.9518, radius=10, center=<-0.2916 -1.2553 0>}
 ]
 thigh_l [{shapeType=Capsule, length=5.0079, radius=10, shapeName=thigh_l_capsule, center=<-7.9291 -0.7152 1.0841>, rotation=<rot 90 -5.0222 0>}]
 thigh_r [{shapeType=Capsule, length=8, radius=10, shapeName=thigh_r_capsule, center=<7.8663 0 -1.0841>, rotation=<rot 90 174.9778 0>}]
}
constraints []
